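# Translucent UI material for the Habanero lobby background, layering triangles, waves and a cracked rift over a color gradient.
# A rotated, scrolling SDF triangle grid gets random per-cell shading, while two animated wave SDFs form a fadeable wave mask.
# A glass-crack texture shaped by radial gradients masks the rift, and stepped UVs sample scrolling noise for pixelation.
# Current and Next colors blend via Color Swap into an invertible radial background with hot-spot and intensity controls.
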Material /Game/UI/Frontend/Lobby/Habanero/Habanero_Materials/M_UI_Habanero_Background
graph {
 properties {
  bUsedWithStaticMesh=false
  blendMode=BLEND_Translucent
  materialDomain=MD_UI
  pixelDepthOffsetMode=PDOM_Legacy
  usedShadingModels=Default Lit
 }
 nodes [
  n1 = MaterialFunctionCall() {materialFunction=&GetUserInterfaceUV}
  n2 = Time()
  n3 = ScalarParameter() {defaultValue=0.01, parameterName=Triangles Offset Speed}
  n4 = Add(
   A=MaterialFunctionCall(
    UVs=MaterialFunctionCall(UV=@n1.Normalized UV) {materialFunction=&MF_UI_UVScaling}.Scaled to Fit
    "Rotation Angle (0-1)"=Constant() {r=0.125}
   ) {materialFunction=&CustomRotator}.Rotated Values
   B=AppendVector(A=Multiply(A=@n2, B=@n3) {constB=0.01}, B=Constant())
  )
  Frac(None=@n4)
  Multiply(A=NamedRerouteUsage(), B=@n4) {constB=10}
  n5 = ScalarParameter() {defaultValue=8, parameterName=Triangles Tiling}
  n6 = Multiply(A=@n4, B=NamedRerouteDeclaration(None=Floor(None=@n5)) {name=Triangles Tiling Value, nodeColor=<rgba 0 0 0 1>}) {constB=10}
  n7 = Ceil(None=MaterialFunctionCall!(
    UVs=Frac(None=@n6)
    Corner 01=Constant2Vector() {g=-1, r=-1}
    Corner 02=Constant2Vector() {g=-1, r=1}
    Corner 03
    Glow Min
    Glow Max
    Outline Thickness
   ) {materialFunction=/FortUILibrary/MaterialFunctions/Shapes/MF_UI_WIP_Triangle}.SDF)
  n8 = Add(A=Constant2Vector() {r=0.5}, B=@n6)
  n9 = MaterialFunctionCall!(UV=Divide(A=Floor(None=@n8), B=NamedRerouteUsage()), Seed=Constant() {r=0.94}) {materialFunction=&MF_UI_Random}
  n10 = OneMinus(None=@n7)
  n11 = MaterialFunctionCall(UV=Floor(None=@n6)) {materialFunction=&MF_UI_Random}
  n12 = NamedRerouteDeclaration(None=Max(A=Multiply(A=@n9.Result, B=@n7), B=Multiply(A=@n10, B=@n11.Result))) {name=Triangles Pattern, nodeColor=<rgba 0 0 0 1>}
  n13 = NamedRerouteUsage()
  n14 = Multiply(A=Multiply(A=@n12, B=@n13)) {constB=0.5}
  NamedRerouteDeclaration(None=Saturate(None=LinearInterpolate(A=@n12, B=@n14, Alpha=@n13))) {name=Triangle Pattern Mask, nodeColor=<rgba 0 0 0 1>}
  n15 = Constant2Vector() {g=0.825}
  n16 = MaterialFunctionCall(
   UVs=MaterialFunctionCall() {materialFunction=&GetUserInterfaceUV}.Normalized UV
   "Rotation Angle (0-1)"=ScalarParameter() {parameterName=Wave Rotation}
  ) {materialFunction=&CustomRotator}
  n17 = Add(A=MaterialFunctionCall(
    UVs=NamedRerouteDeclaration(None=@n16.Rotated Values) {name=Wave UV, nodeColor=<rgba 0 0 0 1>}
    Texture Scale=ScalarParameter() {defaultValue=1, parameterName=Wave Scale}
   ) {materialFunction=&ScaleUVsByCenter}.UVs) {constB=-0.125}
  n18 = ScalarParameter() {defaultValue=0.125, parameterName=Wave Height}
  n19 = ScalarParameter() {defaultValue=0.125, parameterName=Wave Length}
  n20 = ScalarParameter() {defaultValue=1, parameterName=Wave Speed}
  n21 = Constant()
  n22 = Constant2Vector() {g=0.575}
  n23 = AppendVector(
   A=MaterialFunctionCall!(
    UV=Add(A=@n15, B=@n17) {constB=0.825}
    Wave Height=@n18
    Wave Elevation
    Wave Length=@n19
    Wave Speed=@n20
    GlowMin
    GlowMax
    Outline Thickness=@n21
    Outline Alignment
    Custom Curve
   ) {materialFunction=/FortUILibrary/MaterialFunctions/Shapes/MF_UI_Wave}.SDF
   B=MaterialFunctionCall(UV=Add(A=@n17, B=@n22) {constB=0.5625}, Wave Height=@n18, Wave Length=@n19, Wave Speed=@n20, Outline Thickness=@n21) {materialFunction=&MF_UI_Wave}.SDF Outline
  )
  n24 = OneMinus(None=@n23)
  n25 = Saturate(None=Multiply(A=Min(A=@n23, B=@n24)) {constB=2})
  n26 = ComponentMask(None=@n25) {r=true}
  n27 = OneMinus(None=@n26)
  n28 = Multiply(A=Ceil(None=@n26), B=LinearInterpolate(Alpha=Power(Base=@n27)) {constA=0.125})
  n29 = ComponentMask(None=@n25) {g=true}
  n30 = Saturate(None=Add(A=Multiply(A=@n28) {constB=0.25}, B=Power(Base=@n29) {constExponent=16}))
  NamedRerouteDeclaration(None=@n30) {name=Full Wave Mask, nodeColor=<rgba 0 0 0 1>}
  n31 = MaterialFunctionCall() {materialFunction=&GetUserInterfaceUV}
  n32 = Constant() {r=1}
  n33 = VectorParameter() {parameterName=RIft Offset}
  n34 = AppendVector(A=@n33.R, B=@n33.R)
  n35 = MaterialFunctionCall(UV=Add(
    A=MaterialFunctionCall(UVs=@n31.Normalized UV, Texture Scale=@n32) {materialFunction=&ScaleUVsByCenter}.UVs
    B=@n34
   )) {materialFunction=&MF_UI_UVScaling}
  n36 = Subtract(
   A=MaterialFunctionCall(UV=Add(A=@n35.Scaled To Fill, B=@n34)) {materialFunction=&MF_UI_RadialGradient}.Smooth
   B=ScalarParameter() {parameterName=Rift Cracks Removal, sliderMax=1}
  )
  n37 = TextureSample(UVs=@n35.Scaled To Fill) {texture=/Game/Effects/Fort_Effects/Textures/Tile/T_Glass_Cracks_M}
  n38 = Saturate(None=Add(A=@n37.R, B=@n37.R))
  n39 = Saturate(None=Subtract(A=Multiply(A=@n36, B=@n38), B=ScalarParameter() {parameterName=Rift Removal, sliderMax=1}))
  n40 = Multiply(
   A=Multiply(A=Power(Base=@n39)) {constB=32}
   B=MaterialFunctionCall() {materialFunction=&MF_UI_RadialGradient}.Smooth
  )
  NamedRerouteDeclaration(None=LinearInterpolate(A=Multiply(A=@n39, B=@n38), B=@n40, Alpha=@n40)) {name=Cracks Mask, nodeColor=<rgba 0 0 0 1>}
  n41 = NamedRerouteUsage()
  n42 = Subtract(A=OneMinus(None=ComponentMask(None=@n41) {r=true}), B=ScalarParameter() {parameterName=Wave Fade Out, sliderMax=1}) {constB=0}
  NamedRerouteDeclaration(None=Multiply(A=@n30, B=Saturate(None=@n42))) {name=Wave Mask, nodeColor=<rgba 0 0 0 1>}
  n43 = MaterialFunctionCall!("Value (0-1)"=NamedRerouteUsage(), Steps=Multiply(A=NamedRerouteUsage()) {constB=24}) {materialFunction=/FortUILibrary/MaterialFunctions/Basics/MF_UI_StepRatio}
  n44 = Multiply(A=Time(), B=Constant() {r=0.5}) {constB=0.05}
  n45 = AppendVector(A=NamedRerouteDeclaration(None=Frac(None=@n44)) {name=Noise Rotation Speed, nodeColor=<rgba 0 0 0 1>}, B=Constant())
  n46 = TextureObject() {samplerType=SAMPLERTYPE_Grayscale, texture=/FortWater/Textures/LowResBlurredNoise}
  NamedRerouteDeclaration(None=TextureSample(UVs=Subtract(A=@n43.Stepped Ratio, B=@n45), Tex=NamedRerouteDeclaration(None=@n46) {name=Noise Texture, nodeColor=<rgba 0 0 0 1>}) {samplerSource=SSM_Wrap_WorldGroupSettings}.R) {name=Pixelated Noise, nodeColor=<rgba 0 0 0 1>}
  n47 = ScalarParameter() {defaultValue=4, parameterName=Rift Intensity}
  n48 = Multiply(
   A=OneMinus(None=NamedRerouteUsage())
   B=Multiply(A=NamedRerouteUsage(), B=ScalarParameter() {defaultValue=1, parameterName=Triangle Opacity, sliderMax=1}) {constB=0.5}
  )
  n49 = MaterialFunctionCall(
   A=VectorParameter() {parameterName=Current Color, defaultValue=<rgba 1 0.3547 0 1>}.RGB
   B=Constant() {r=1}
   C=VectorParameter() {parameterName=Next Color, defaultValue=<rgba 0.0963 0 0.625 1>}.RGB
   Alpha=ScalarParameter() {parameterName=Color Swap, sliderMax=1}
  ) {materialFunction=&3ColorBlend}
  n50 = MaterialFunctionCall(UV=MaterialFunctionCall(
    UVs=MaterialFunctionCall() {materialFunction=&GetUserInterfaceUV}.Normalized UV
    Texture Scale=ScalarParameter() {defaultValue=1.2, parameterName=Spot Scale}
   ) {materialFunction=&ScaleUVsByCenter}.UVs) {materialFunction=&MF_UI_UVScaling}
  n51 = NamedRerouteDeclaration(None=LinearInterpolate(
    A=MaterialFunctionCall(UV=@n50.Scaled To Fill) {materialFunction=&MF_UI_RadialGradient}.Smooth
    Alpha=ScalarParameter() {parameterName=Radial Background Fade Out, sliderMax=1}
   )) {name=Radial Background Mask, nodeColor=<rgba 0 0 0 1>}
  n52 = LinearInterpolate(
   A=Multiply(A=@n49.Result) {constB=0.02, desc=Darken the color}
   B=@n49.Result
   Alpha=LinearInterpolate(A=@n51, B=OneMinus(None=@n51), Alpha=ScalarParameter() {defaultValue=1, parameterName=Invert Radial, sliderMax=1})
  )
  n53 = Multiply(A=NamedRerouteUsage(), B=ScalarParameter() {parameterName=Hot Spot Intensity}) {constB=0.5}
  n54 = Multiply(A=@n52, B=Add(A=@n52, B=@n53))
  n55 = Saturate(None=Multiply(A=OneMinus(None=@n48), B=LinearInterpolate(A=@n52, B=@n54, Alpha=@n53)))
  n56 = Add(A=Multiply(A=@n47, B=NamedRerouteUsage()), B=@n55)
  Saturate(None=Add(A=LinearInterpolate(A=@n55, B=@n56, Alpha=@n47), B=@n55))
  ScalarParameter() {defaultValue=1, parameterName=Background Opacity, sliderMax=1}
 ]
}
