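# Additive UI material that draws an animated background of scrolling, randomized triangles for the Habanero lobby frontend.
# A mirrored, tiled particle grid scrolls over time with randomized speed and cell scale, optionally masking the triangles' center.
# Each cell's UV is rotated in pseudo-3D, cut into a hard triangle by an SDF, and thinned by a remover parameter.
# An eased radial mask fades the pattern, with parameters for color, brightness and opacity.
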Material /Game/UI/Frontend/Lobby/Habanero/Habanero_Materials/M_UI_Habanero_Background_Triangles
graph {
 properties {bUsedWithStaticMesh=false, blendMode=BLEND_Additive, materialDomain=MD_UI, pixelDepthOffsetMode=PDOM_Legacy, usedShadingModels=Default Lit}
 nodes [
  n1 = Saturate(None=Add(A=NamedRerouteUsage()) {constB=0.4})
  n2 = VectorParameter() {parameterName=Color, defaultValue=<rgba 1 0.6362 0.1389 1>}
  LinearInterpolate(
   A=StaticSwitchParameter(True=Multiply(A=@n1, B=@n2.RGB), False=@n2.RGB) {defaultValue=true, parameterName=Randomize More Properties}
   Alpha=ScalarParameter() {parameterName=Color Brightness, sliderMax=1}
  )
  n3 = NamedRerouteDeclaration(None=MaterialFunctionCall() {materialFunction=&MF_UI_RadialGradient}.Smooth) {name=Radial Mask, nodeColor=<rgba 0 0 0 1>}
  n4 = NamedRerouteUsage()
  n5 = NamedRerouteUsage()
  Multiply(
   A=Multiply(
    A=MaterialFunctionCall("Value (0-1)"=@n3) {materialFunction=&MF_UI_EaseCurves}.EaseInSine
    B=Multiply(A=@n4, B=@n5)
   )
   B=ScalarParameter() {defaultValue=1, parameterName=Opacity, sliderMax=1}
  )
  n6 = MaterialFunctionCall() {materialFunction=&GetUserInterfaceUV}
  n7 = Constant()
  n8 = Constant2Vector() {g=1, r=2}
  n9 = Multiply(A=Multiply(A=Time(), B=ScalarParameter() {defaultValue=1, parameterName=Triangles Speed}), B=Add(A=NamedRerouteUsage()) {constB=0.5})
  n10 = Multiply(
   A=MaterialFunctionCall("Value (0-1)"=NamedRerouteUsage()) {materialFunction=&MF_UI_EaseCurves}.EaseOutCubic
   B=ScalarParameter() {defaultValue=0.5, parameterName=Triangle Size}
  ) {constB=0.5}
  n11 = Multiply(A=@n10, B=NamedRerouteUsage())
  n12 = NamedRerouteUsage()
  n13 = Subtract(A=@n12) {constB=0.25}
  n14 = Multiply(A=@n11, B=LinearInterpolate(A=Saturate(None=@n13), B=@n12, Alpha=NamedRerouteUsage()) {constA=0.25})
  n15 = MaterialFunctionCall!(
   UV=MaterialFunctionCall(UV=MaterialFunctionCall!(UV=@n6.Normalized UV, Offset=@n7, Rotation=@n7, Scale=@n8, MirrorY) {materialFunction=/FortUILibrary/MaterialFunctions/Basics/MF_UI_Symmetry_2x1_DEPRECATED}.Result) {materialFunction=&MF_UI_UVScaling}.Scaled to Fit
   Global Offset=AppendVector(A=Frac(None=@n9), B=Constant())
   Global Rotation
   Tiling=NamedRerouteDeclaration(None=ScalarParameter() {defaultValue=4, parameterName=Triangles Tiling}) {name=Triangles Tiling, nodeColor=<rgba 0 0 0 1>}
   Cell Offset=Constant()
   Cell Scale=StaticSwitchParameter(True=StaticSwitchParameter(True=@n14, False=@n11) {defaultValue=true, parameterName=Randomize More Properties}, False=@n10) {defaultValue=true, parameterName=Mask Triangles Center}
   Cell Rotation
   Random Seed=ScalarParameter() {defaultValue=0.2525, parameterName=Random Seed}
  ) {materialFunction=/FortUILibrary/MaterialFunctions/Basics/MF_UI_ParticleGrid}
  NamedRerouteDeclaration(None=@n15.Random Rows) {name=Random Rows, nodeColor=<rgba 0 0 0 1>}
  n16 = NamedRerouteUsage()
  MaterialExpressionAppend3Vector_2 = MaterialXAppend3Vector!(A=Constant(), B=@n16, C=@n16)
  n17 = MaterialFunctionCall!(
   UV=@n15.Cells UV
   Angle=@MaterialExpressionAppend3Vector_2
   Camera FOV
   Camera Offset=Constant3Vector() {constant=<rgba 0 0 2.41 0>}
   Offset from Pivot
   Texture Scale
   Texture Size
  ) {materialFunction=/FortUILibrary/MaterialFunctions/Basics/MF_UI_Rotate3D}
  n18 = Constant2Vector() {g=-1, r=-1}
  n19 = Constant2Vector() {g=-1, r=1}
  n20 = OneMinus(None=Ceil(None=MaterialFunctionCall!(UVs=@n17.Rotated UV, Corner 01=@n18, Corner 02=@n19, Corner 03, Glow Min, Glow Max, Outline Thickness) {materialFunction=/FortUILibrary/MaterialFunctions/Shapes/MF_UI_WIP_Triangle}.SDF))
  n21 = Saturate(None=Subtract(A=NamedRerouteUsage(), B=ScalarParameter() {parameterName=Triangles Remover, sliderMax=1}) {constB=0.6})
  NamedRerouteDeclaration(None=Multiply(A=Saturate(None=@n20), B=Ceil(None=@n21))) {name=Triangle Pattern Mask, nodeColor=<rgba 0 0 0 1>}
  n22 = ScalarParameter() {defaultValue=1, parameterName=Triangle Mask Scale}
  n23 = MaterialFunctionCall(UV=MaterialFunctionCall(UV=MaterialFunctionCall(Texture Scale=@n22) {materialFunction=&ScaleUVsByCenter}.UVs) {materialFunction=&MF_UI_UVScaling}.Scaled to Fit) {materialFunction=&MF_UI_RadialGradient}
  NamedRerouteDeclaration(None=MaterialFunctionCall("Value (0-1)"=OneMinus(None=@n23.Smooth)) {materialFunction=&MF_UI_EaseCurves}.EaseInCubic) {name=Radial Scale Mask, nodeColor=<rgba 0 0 0 1>}
  n24 = NamedRerouteDeclaration(None=@n15.Random Cell Values) {name=Random Cells, nodeColor=<rgba 0 0 0 1>}
  n25 = Multiply(A=Add(A=Sine(None=@n24))) {constB=0.5}
  n26 = MaterialFunctionCall() {materialFunction=&GetUserInterfaceUV}
  NamedRerouteDeclaration(None=DotProduct(A=NamedRerouteDeclaration(None=@n25) {name=Sine Random Cells, nodeColor=<rgba 0 0 0 1>}, B=ComponentMask(None=@n26.Normalized UV) {r=true})) {name=Assymetric Random Cells, nodeColor=<rgba 0 0 0 1>}
  StaticSwitchParameter(True=NamedRerouteUsage(), False=@n5) {parameterName=Randomize Opacity}
 ]
}
