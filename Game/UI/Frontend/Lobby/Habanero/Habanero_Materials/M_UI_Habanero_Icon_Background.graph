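# Translucent UI material that draws an animated cloud-and-stars background for Habanero icons in the frontend lobby.
# A rotating cloud texture, distorted by its own grayscale flow map, combines with a radial gradient into cloud and highlight masks.
# A scrolling star texture gains a time-driven twinkle.
# Masks shade and brighten Inner and Outer color parameters, with parameters also exposed for scale, rotation offset and cloud opacity.
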Material /Game/UI/Frontend/Lobby/Habanero/Habanero_Materials/M_UI_Habanero_Icon_Background
graph {
 properties {
  bUsedWithStaticMesh=false
  blendMode=BLEND_Translucent
  materialDomain=MD_UI
  pixelDepthOffsetMode=PDOM_Legacy
  usedShadingModels=Default Lit
 }
 nodes [
  n1 = NamedRerouteUsage()
  n2 = Multiply(A=@n1, B=Multiply(A=NamedRerouteUsage()) {constB=0.25})
  Multiply(A=Saturate(None=Add(A=@n1, B=@n2)), B=ScalarParameter() {defaultValue=1, parameterName=Cloud Opacity, sliderMax=1})
  n3 = Constant()
  n4 = Frac(None=Multiply(A=Time()) {constB=0.02})
  n5 = TextureSample(
   UVs=Add(A=NamedRerouteUsage(), B=AppendVector(A=@n3, B=@n4))
   Tex=TextureObject() {samplerType=SAMPLERTYPE_Alpha, texture=/Game/UI/Frontend/Lobby/Habanero/Habanero_Textures/T_UI_Habanero_Clouds_Stars}
  )
  n6 = Saturate(None=Subtract(A=@n5.R) {constB=0.4})
  n7 = Frac(None=Multiply(A=Time()) {constB=0.5})
  n8 = Min(A=@n7, B=OneMinus(None=@n7))
  n9 = Multiply(A=Saturate(None=Subtract(A=@n5.R) {constB=0.1}), B=LinearInterpolate(Alpha=Multiply(A=@n8) {constB=2}) {constB=0.75})
  NamedRerouteDeclaration(None=Saturate(None=Add(A=@n6, B=@n9))) {name=Stars, nodeColor=<rgba 0 0 0 1>}
  NamedRerouteDeclaration(None=MaterialFunctionCall(
    UVs=MaterialFunctionCall() {materialFunction=&GetUserInterfaceUV}.Normalized UV
    Texture Scale=ScalarParameter() {defaultValue=1, parameterName=Scale}
   ) {materialFunction=&ScaleUVsByCenter}.UVs) {name=Common UV, nodeColor=<rgba 0 0 0 1>}
  n10 = MaterialFunctionCall("Value (0-1)"=MaterialFunctionCall(GradientStart=Constant(), GradientEnd=Constant() {r=1}) {materialFunction=&MF_UI_RadialGradient}.Smooth) {materialFunction=&MF_UI_EaseCurves}
  n11 = Frac(None=Multiply(A=Time()) {constB=0.025})
  n12 = OneMinus(None=@n11)
  n13 = ScalarParameter() {parameterName=Base Rotation Factor}
  n14 = TextureObject() {samplerType=SAMPLERTYPE_Alpha, texture=/Game/UI/Frontend/Lobby/Habanero/Habanero_Textures/T_UI_Habanero_Clouds}
  n15 = MaterialFunctionCall!(
   UVs=MaterialFunctionCall(UVs=NamedRerouteUsage(), "Rotation Angle (0-1)"=Add(A=@n12, B=@n13)) {materialFunction=&CustomRotator}.Rotated Values
   Texture Object=@n14
   Grayscale Flow Map=TextureSample(Tex=@n14).RGB
   Flow Direction=@n11
   Flow Intensity=Constant() {r=0.25}
   Flow Speed
  ) {materialFunction=/FortUILibrary/MaterialFunctions/Effects/MF_UI_GrayscaleFlowMap_New}
  n16 = Saturate(None=Subtract(A=@n15.Distorted Texture) {constB=0.1})
  n17 = Multiply(A=MaterialFunctionCall("Value (0-1)"=Multiply(A=@n16, B=@n15.Distorted Texture)) {materialFunction=&MF_UI_EaseCurves}.EaseInCubic) {constB=0.75}
  n18 = NamedRerouteDeclaration(None=Saturate(None=Add(A=@n10.EaseInSine, B=@n17))) {name=Cloud Mask, nodeColor=<rgba 0 0 0 1>}
  n19 = Abs(None=ConstantBiasScale(None=Multiply(A=@n18) {constB=2}) {bias=-0.5, scale=2})
  NamedRerouteDeclaration(None=Power(Base=OneMinus(None=@n19))) {name=Highlight Cloud Mask, nodeColor=<rgba 0 0 0 1>}
  n20 = NamedRerouteUsage()
  n21 = LinearInterpolate(
   A=VectorParameter() {parameterName=Outer Color, defaultValue=<rgba 1 0.4233 0 1>}.RGB
   B=VectorParameter() {parameterName=Inner Color, defaultValue=<rgba 0.9131 0.2462 0.0844 1>}.RGB
   Alpha=@n20
  )
  n22 = Multiply(A=NamedRerouteUsage(), B=@n21)
  n23 = Multiply(A=Multiply(A=@n21, B=@n20)) {constB=2}
  n24 = Add(A=Add(A=@n22, B=@n22), B=Saturate(None=Add(A=@n21, B=@n23)))
  n25 = Add(A=@n24, B=@n24)
  Saturate(None=LinearInterpolate(A=@n21, B=Saturate(None=@n25), Alpha=NamedRerouteUsage()))
 ]
}
