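# Translucent UI material for the foreground of the Habanero lobby tier icon, drawing a tinted gem and crest from packed-channel textures.
# Channels split into glow, lighting, color and silhouette maps, colored by parameters with optional asymmetric gem coloring.
# Static switches select tier layouts, including four crystals cut from a 2x2 atlas with offset UVs.
# Time drives pulsing glow, floating, spinning light and shine, while hotness and desaturate parameters adjust the result.
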Material /Game/UI/Frontend/Lobby/Habanero/Habanero_Materials/M_UI_Habanero_Icon_Foreground
graph {
 properties {
  bUsedWithStaticMesh=false
  blendMode=BLEND_Translucent
  materialDomain=MD_UI
  pixelDepthOffsetMode=PDOM_Legacy
  usedShadingModels=Default Lit
 }
 nodes [
  n1 = NamedRerouteUsage()
  n2 = Frac(None=Multiply(A=Time()) {constB=0.25})
  n3 = Min(A=@n2, B=OneMinus(None=@n2))
  n4 = Multiply(A=NamedRerouteUsage(), B=LinearInterpolate(B=Constant() {desc=Glow Max Opacity, r=1}, Alpha=Multiply(A=@n3) {constB=2}) {constA=0.5})
  n5 = Add(
   A=NamedRerouteUsage()
   B=Multiply(
    A=NamedRerouteDeclaration(None=@n4) {name=Glow Opacity, nodeColor=<rgba 0 0 0 1>}
    B=VectorParameter() {group=1 - Colors, parameterName=Gem Glow Color, defaultValue=<rgba 0.0896 0.7548 0.8879 1>}.RGB
   )
  )
  n6 = NamedRerouteUsage()
  n7 = ScalarParameter() {parameterName=Gem Hotness}
  n8 = Add(
   A=LinearInterpolate(A=Saturate(None=@n5), B=NamedRerouteUsage(), Alpha=NamedRerouteUsage())
   B=Multiply(A=NamedRerouteUsage(), B=Multiply(A=@n6, B=@n7) {desc=Top Glow})
  )
  n9 = ScalarParameter() {parameterName=Badge Hotness}
  n10 = LinearInterpolate(A=@n8, B=Add(A=@n8, B=Multiply(A=@n8, B=@n9)) {constB=2}, Alpha=@n9)
  n11 = Add(A=LinearInterpolate(B=@n10, Alpha=Multiply(A=@n1, B=@n10)), B=@n10)
  LinearInterpolate(A=@n11, B=Desaturation(None=@n11), Alpha=ScalarParameter() {parameterName=Desaturate})
  n12 = NamedRerouteUsage()
  n13 = ScalarParameter() {defaultValue=1, parameterName=Glow Opacity, sliderMax=1}
  n14 = NamedRerouteUsage()
  n15 = NamedRerouteUsage()
  n16 = ScalarParameter() {defaultValue=0.75, parameterName=Shards Translucency, sliderMax=1}
  Max(
   A=Max(A=Multiply(A=@n12, B=@n13), B=LinearInterpolate(A=@n14, B=@n15, Alpha=@n16))
   B=Multiply(A=NamedRerouteUsage(), B=ScalarParameter() {defaultValue=1, parameterName=Crest Opacity, sliderMax=1})
  )
  n17 = NamedRerouteUsage()
  n18 = ScalarParameter() {defaultValue=1, parameterName=Gem Scale}
  n19 = TextureObjectParameter() {
   group=0 - Texture
   parameterName=Gem Texture
   samplerType=SAMPLERTYPE_LinearColor
   sortPriority=0
   texture=/Game/UI/Frontend/Lobby/Habanero/Habanero_Textures/HTiers/T_UI_Habanero_08_Gem
  }
  n20 = TextureSample(
   UVs=NamedRerouteDeclaration(None=MaterialFunctionCall(UVs=@n17, Texture Scale=@n18) {materialFunction=&ScaleUVsByCenter}.UVs) {name=Default Gem UV, nodeColor=<rgba 0 0 0 1>}
   Tex=NamedRerouteDeclaration(None=@n19) {name=Gem Texture, nodeColor=<rgba 0 0 0 1>}
  ) {samplerSource=SSM_Clamp_WorldGroupSettings, samplerType=SAMPLERTYPE_LinearColor}
  n21 = NamedRerouteUsage()
  n22 = MaterialFunctionCall(Float4=StaticSwitchParameter(True=@n20.RGBA, False=StaticSwitchParameter(True=@n21, False=@n20.RGBA) {defaultValue=true, parameterName=Highest Tier}) {parameterName=Default Gem}) {materialFunction=&BreakOutFloat4Components}
  NamedRerouteDeclaration(None=@n22.B) {name=Gem Color Map, nodeColor=<rgba 0 0 0 1>}
  n23 = NamedRerouteDeclaration(None=VectorParameter() {group=1 - Colors, parameterName=Crest Color, defaultValue=<rgba 0.3231 0.0666 0.0284 1>}.RGB) {name=Crest Color, nodeColor=<rgba 0 0 0 1>}
  n24 = Multiply(A=NamedRerouteUsage(), B=NamedRerouteUsage())
  n25 = NamedRerouteUsage()
  n26 = Power(Base=@n25) {constExponent=8}
  n27 = Power(Base=OneMinus(None=@n25)) {constExponent=4}
  n28 = Multiply(
   A=VectorParameter() {group=1 - Colors, parameterName=Crest Highlight Color, defaultValue=<rgba 0.413 0.7312 1 1>}.RGB
   B=Multiply(A=Min(A=@n26, B=@n27)) {constB=8}
  )
  n29 = LinearInterpolate(A=@n23, B=LinearInterpolate(A=@n23, B=Add(A=@n24, B=@n28), Alpha=NamedRerouteUsage()), Alpha=NamedRerouteUsage())
  n30 = Power(Base=NamedRerouteUsage(), Exp=Constant() {r=6}) {constExponent=6}
  n31 = LinearInterpolate(B=@n29, Alpha=Multiply(A=Saturate(None=@n30), B=ScalarParameter() {defaultValue=8, parameterName=Crest Brightness})) {desc=Crest Highlight}
  n32 = Multiply(A=@n29) {constB=0.8, desc=Darken Crest}
  NamedRerouteDeclaration(None=Saturate(None=Add(A=@n31, B=@n32))) {name=Crest Visual, nodeColor=<rgba 0 0 0 1>}
  n33 = NamedRerouteUsage()
  n34 = NamedRerouteUsage()
  n35 = NamedRerouteUsage()
  n36 = LinearInterpolate(
   A=LinearInterpolate(A=LinearInterpolate(A=@n33, B=@n34, Alpha=@n35), B=NamedRerouteUsage(), Alpha=NamedRerouteUsage())
   B=NamedRerouteUsage()
   Alpha=NamedRerouteUsage()
  )
  NamedRerouteDeclaration(None=ComponentMask(None=@n36) {r=true}) {name=Tier 3 Gems Glow Map, nodeColor=<rgba 0 0 0 1>}
  NamedRerouteDeclaration(None=ComponentMask(None=@n36) {b=true}) {name=Tier 3 Gems Color Map, nodeColor=<rgba 0 0 0 1>}
  n37 = Frac(None=Multiply(A=Time()) {constB=0.25})
  n38 = Min(A=@n37, B=OneMinus(None=@n37))
  NamedRerouteDeclaration(None=MaterialFunctionCall("Value (0-1)"=Multiply(A=@n38) {constB=2}) {materialFunction=&MF_UI_EaseCurves}.EaseInOutSine) {name=Floating Time, nodeColor=<rgba 0 0 0 1>}
  n39 = Multiply(A=AppendVector(A=@n22.R, B=@n22.R), B=ScalarParameter() {defaultValue=1, parameterName=Gems Opacity, sliderMax=1})
  NamedRerouteDeclaration(None=ComponentMask(None=@n39) {r=true}) {name=Gem Glow, nodeColor=<rgba 0 0 0 1>}
  NamedRerouteDeclaration(None=ComponentMask(None=@n39) {g=true}) {name=Gem Silhouette, nodeColor=<rgba 0 0 0 1>}
  NamedRerouteDeclaration(None=@n22.G) {name=Gem Lighting Map, nodeColor=<rgba 0 0 0 1>}
  NamedRerouteDeclaration(None=ComponentMask(None=@n36) {g=true}) {name=Tier 3 Gems Lighting Map, nodeColor=<rgba 0 0 0 1>}
  n40 = Max(A=NamedRerouteUsage(), B=NamedRerouteUsage())
  n41 = NamedRerouteUsage()
  NamedRerouteDeclaration(None=Max(A=Max(A=@n40, B=@n41), B=NamedRerouteUsage())) {name=Tier 3 Gems Silhouette, nodeColor=<rgba 0 0 0 1>}
  MaterialExpressionAppend4Vector_0 = MaterialXAppend4Vector!(A=NamedRerouteUsage(), B=NamedRerouteUsage(), C=NamedRerouteUsage(), D=NamedRerouteUsage())
  NamedRerouteDeclaration(None=@MaterialExpressionAppend4Vector_0) {name=Tier 3 Gems Channels, nodeColor=<rgba 0 0 0 1>}
  n42 = VectorParameter() {group=1 - Colors, parameterName=Secondary Gem Color, defaultValue=<rgba 0.1022 0.8228 0.9911 1>}
  n43 = LinearInterpolate(
   A=LinearInterpolate(
    A=NamedRerouteUsage()
    B=NamedRerouteDeclaration(None=@n42.RGB) {name=Gem Secondary Color, nodeColor=<rgba 0 0 0 1>}
    Alpha=NamedRerouteUsage()
   )
   B=NamedRerouteDeclaration(None=VectorParameter() {group=1 - Colors, parameterName=Gem Color, defaultValue=<rgba 0.8333 0.3379 0 1>}.RGB) {name=Gem Primary Color, nodeColor=<rgba 0 0 0 1>}
   Alpha=NamedRerouteUsage()
  )
  n44 = NamedRerouteUsage()
  n45 = NamedRerouteUsage()
  n46 = ScalarParameter() {group=10 - Color Symmetry, parameterName=Color Symmetry Swap, sliderMax=1, sortPriority=2}
  n47 = ComponentMask(None=MaterialFunctionCall() {materialFunction=&GetUserInterfaceUV}.Normalized UV) {r=true}
  n48 = ScalarParameter() {defaultValue=0.5, group=10 - Color Symmetry, parameterName=Color Assymetry Intensity, sliderMax=1, sortPriority=1}
  n49 = StaticSwitchParameter(True=LinearInterpolate(A=@n43, B=LinearInterpolate(A=@n44, B=@n45, Alpha=@n46), Alpha=Multiply(A=@n47, B=@n48) {constB=0.5}), False=@n43) {group=10 - Color Symmetry, parameterName=Break Gem Color Symmetry, sortPriority=0}
  n50 = Saturate(None=Power(Base=NamedRerouteUsage(), Exp=ScalarParameter() {defaultValue=8, parameterName=Gem Power}) {constExponent=6})
  n51 = ScalarParameter() {defaultValue=0.75, parameterName=Gem Brightness}
  n52 = LinearInterpolate(B=@n49, Alpha=Multiply(A=Multiply(A=@n50, B=@n51) {constB=0.75}, B=NamedRerouteUsage())) {desc=Gem Highlight}
  NamedRerouteDeclaration(None=Saturate(None=Add(A=@n52, B=@n49))) {name=Gem Visual, nodeColor=<rgba 0 0 0 1>}
  n53 = NamedRerouteUsage()
  n54 = Divide(A=Subtract(A=Multiply(A=@n53) {constB=0.5}) {constB=0.25}, B=Constant() {r=0.5})
  n55 = LinearInterpolate(Alpha=NamedRerouteUsage()) {constA=-0.005, constB=0.005}
  n56 = Add(A=Add(A=Add(A=@n54) {constB=0.25}, B=Constant2Vector()) {constB=2}, B=AppendVector(A=Constant(), B=Add(A=@n55) {constB=0.05}))
  n57 = NamedRerouteUsage()
  n58 = Multiply(
   A=MaterialFunctionCall!(UV=NamedRerouteUsage(), Grid Size=Constant() {r=2}, Index=Constant2Vector() {g=1, r=1}) {materialFunction=/FortUILibrary/MaterialFunctions/Basics/MF_UI_IsolateUV}.Result
   B=TextureSample(UVs=NamedRerouteDeclaration(None=@n56) {name=Gem 01 UV, nodeColor=<rgba 0 0 0 1>}, Tex=@n57) {samplerSource=SSM_Clamp_WorldGroupSettings, samplerType=SAMPLERTYPE_LinearColor}.RGBA
  )
  NamedRerouteDeclaration(None=ComponentMask(None=@n58) {a=true}) {name=Crystal 01 Mask, nodeColor=<rgba 0 0 0 1>}
  n59 = TextureObjectParameter() {
   group=0 - Texture
   parameterName=Crest Texture
   samplerType=SAMPLERTYPE_LinearColor
   sortPriority=1
   texture=/Game/UI/Frontend/Lobby/Habanero/Habanero_Textures/HTiers/T_UI_Habanero_08_Crest
  }
  n60 = TextureSample(
   UVs=MaterialFunctionCall(UVs=NamedRerouteUsage(), Texture Scale=ScalarParameter() {defaultValue=1, parameterName=Crest Scale}) {materialFunction=&ScaleUVsByCenter}.UVs
   Tex=NamedRerouteDeclaration(None=@n59) {name=Crest Texture, nodeColor=<rgba 0 0 0 1>}
  ) {samplerSource=SSM_Clamp_WorldGroupSettings}
  n61 = StaticSwitchParameter(True=@n60.RGBA, False=@n60.RGBA) {defaultValue=true, parameterName=Highest Tier}
  n62 = MaterialFunctionCall(Float2=StaticSwitchParameter(True=StaticSwitchParameter(True=@n60.RGBA, False=@n61) {parameterName=Default Gem}, False=Constant()) {defaultValue=true, parameterName=Display Crest}) {materialFunction=&BreakOutFloat2Components}
  NamedRerouteDeclaration(None=@n62.R) {name=Crest, nodeColor=<rgba 0 0 0 1>}
  n63 = NamedRerouteUsage()
  n64 = Multiply(A=MaterialFunctionCall!(UVs=@n63, Alpha) {materialFunction=/FortUILibrary/MaterialFunctions/Basics/MF_UI_ClipEdgePixel}.Result, B=@n63)
  n65 = MaterialFunctionCall!(UV=@n64, Offset=Constant2Vector() {g=-0.05, r=-0.2}, Rotation, Scale=Constant2Vector() {g=1, r=-2}, MirrorY) {materialFunction=/FortUILibrary/MaterialFunctions/Basics/MF_UI_Symmetry_2x1_DEPRECATED}
  n66 = Divide(A=Subtract(A=Multiply(A=@n65.Result) {constB=0.5}) {constB=0.25}, B=Constant() {r=0.5})
  n67 = NamedRerouteUsage()
  n68 = NamedRerouteDeclaration(None=Round(None=ComponentMask(None=@n67) {r=true})) {name=Split UV, nodeColor=<rgba 0 0 0 1>}
  n69 = LinearInterpolate(
   A=LinearInterpolate(Alpha=@n68) {constA=-0.04, constB=0.04}
   B=LinearInterpolate(Alpha=@n68) {constA=-0.02, constB=0.02}
   Alpha=NamedRerouteUsage()
  ) {constA=-0.02, constB=0.02}
  n70 = Add(A=Add(A=Add(A=@n66) {constB=0.25}, B=Constant2Vector() {r=0.5}) {constB=2}, B=AppendVector(A=Constant(), B=Add(A=@n69) {constB=0.05}))
  n71 = Multiply(
   A=MaterialFunctionCall(UV=NamedRerouteUsage(), Grid Size=Constant() {r=2}, Index=Constant2Vector() {g=1, r=2}) {materialFunction=&MF_UI_IsolateUV}.Result
   B=TextureSample(UVs=NamedRerouteDeclaration(None=@n70) {name=Gem 02 UV, nodeColor=<rgba 0 0 0 1>}, Tex=@n57) {samplerSource=SSM_Clamp_WorldGroupSettings, samplerType=SAMPLERTYPE_LinearColor}.RGBA
  )
  NamedRerouteDeclaration(None=ComponentMask(None=@n71) {a=true}) {name=Crystal 02 Mask, nodeColor=<rgba 0 0 0 1>}
  n72 = MaterialFunctionCall(UV=@n64, Offset=Constant2Vector() {g=-0.05}, Scale=Constant2Vector() {g=1, r=-2}) {materialFunction=&MF_UI_Symmetry_2x1_DEPRECATED}
  n73 = Divide(A=Subtract(A=Multiply(A=@n72.Result) {constB=0.5}) {constB=0.25}, B=Constant() {r=0.5})
  n74 = Constant2Vector() {g=0.02, r=-0.01}
  n75 = Constant2Vector() {g=0.01, r=-0.02}
  n76 = NamedRerouteUsage()
  n77 = LinearInterpolate(
   A=LinearInterpolate(A=@n74.R, B=@n75.R, Alpha=@n76) {constB=0.01}
   B=LinearInterpolate(A=@n74.G, B=@n75.G, Alpha=@n76) {constA=-0.01, constB=0.1}
   Alpha=NamedRerouteUsage()
  ) {constA=-0.01, constB=0.01}
  n78 = Add(A=Add(A=Add(A=@n73) {constB=0.25}, B=Constant2Vector() {g=0.5}) {constB=2}, B=AppendVector(A=Constant(), B=Add(A=@n77) {constB=0.05}))
  n79 = Multiply(
   A=MaterialFunctionCall(UV=NamedRerouteUsage(), Grid Size=Constant() {r=2}, Index=Constant2Vector() {g=2, r=1}) {materialFunction=&MF_UI_IsolateUV}.Result
   B=TextureSample(UVs=NamedRerouteDeclaration(None=@n78) {name=Gem 03 UV, nodeColor=<rgba 0 0 0 1>}, Tex=@n57) {samplerSource=SSM_Clamp_WorldGroupSettings, samplerType=SAMPLERTYPE_LinearColor}.RGBA
  )
  NamedRerouteDeclaration(None=ComponentMask(None=@n79) {a=true}) {name=Crystal 03 Mask, nodeColor=<rgba 0 0 0 1>}
  n80 = MaterialFunctionCall(UV=@n64, Offset=Constant2Vector() {g=-0.1, r=0.15}, Scale=Constant2Vector() {g=1, r=-2}) {materialFunction=&MF_UI_Symmetry_2x1_DEPRECATED}
  n81 = Divide(A=Subtract(A=Multiply(A=@n80.Result) {constB=0.5}) {constB=0.25}, B=Constant() {r=0.5})
  n82 = NamedRerouteUsage()
  n83 = LinearInterpolate(
   A=LinearInterpolate(Alpha=@n82) {constA=-0.01, constB=0.03}
   B=LinearInterpolate(Alpha=@n82) {constA=-0.02, constB=0.02}
   Alpha=NamedRerouteUsage()
  ) {constA=-0.02, constB=0.02}
  n84 = Add(A=Add(A=Add(A=@n81) {constB=0.25}, B=Constant2Vector() {g=0.5, r=0.5}) {constB=2}, B=AppendVector(A=Constant(), B=Add(A=@n83) {constB=0.05}))
  n85 = Multiply(
   A=MaterialFunctionCall(UV=NamedRerouteUsage(), Grid Size=Constant() {r=2}, Index=Constant2Vector() {g=2, r=2}) {materialFunction=&MF_UI_IsolateUV}.Result
   B=TextureSample(UVs=NamedRerouteDeclaration(None=@n84) {name=Gem 04 UV, nodeColor=<rgba 0 0 0 1>}, Tex=@n57) {samplerSource=SSM_Clamp_WorldGroupSettings, samplerType=SAMPLERTYPE_LinearColor}.RGBA
  )
  NamedRerouteDeclaration(None=ComponentMask(None=@n85) {a=true}) {name=Crystal 04 Mask, nodeColor=<rgba 0 0 0 1>}
  NamedRerouteDeclaration(None=ComponentMask(None=@n85) {b=true, g=true, r=true}) {name=Crystal 04 Maps, nodeColor=<rgba 0 0 0 1>}
  NamedRerouteDeclaration(None=ComponentMask(None=@n79) {b=true, g=true, r=true}) {name=Crystal 03 Maps, nodeColor=<rgba 0 0 0 1>}
  NamedRerouteDeclaration(None=ComponentMask(None=@n58) {b=true, g=true, r=true}) {name=Crystal 01 Maps, nodeColor=<rgba 0 0 0 1>}
  NamedRerouteDeclaration(None=ComponentMask(None=@n71) {b=true, g=true, r=true}) {name=Crystal 02 Maps, nodeColor=<rgba 0 0 0 1>}
  n86 = Time()
  n87 = MaterialFunctionCall(
   UVs=MaterialFunctionCall() {materialFunction=&GetUserInterfaceUV}.Normalized UV
   Rotation Center=Constant2Vector() {g=0.25, r=0.25}
   "Rotation Angle (0-1)"=Frac(None=Multiply(A=@n86) {constB=0.1})
  ) {materialFunction=&CustomRotator}
  NamedRerouteDeclaration(None=LinearInterpolate(Alpha=MaterialFunctionCall(UV=@n87.Rotated Values) {materialFunction=&MF_UI_RadialGradient}.Smooth) {constA=0.25}) {name=Lighting Spin, nodeColor=<rgba 0 0 0 1>}
  NamedRerouteDeclaration(None=@n62.G) {name=Crest Silhouette, nodeColor=<rgba 0 0 0 1>}
  n88 = Multiply(A=MaterialFunctionCall(
    UVs=MaterialFunctionCall() {materialFunction=&GetUserInterfaceUV}.Normalized UV
    "Rotation Angle (0-1)"=Constant() {r=-0.125}
   ) {materialFunction=&CustomRotator}.Rotated Values) {constB=0.5}
  n89 = MaterialFunctionCall("Value (0-1)"=MaterialFunctionCall!(Animation Duration=Constant() {r=2}, Delay=Constant() {r=2}, Reset to Default) {materialFunction=/FortUILibrary/MaterialFunctions/Time/MF_UI_AnimationDelay}.Result) {materialFunction=&MF_UI_EaseCurves}
  n90 = Frac(None=Subtract(A=Add(A=@n88), B=Subtract(A=@n89.EaseOutCubic) {constB=0.5}))
  n91 = Constant() {r=0.1}
  MaterialExpressionRemap_4 = MaterialXRemap!(None=Subtract(A=Sine(None=@n90), B=OneMinus(None=@n91)) {constB=0.5}, InputLow, InputHigh=@n91, TargetLow, TargetHigh)
  NamedRerouteDeclaration(None=ComponentMask(None=Saturate(None=@MaterialExpressionRemap_4)) {r=true}) {name=Shining Effect, nodeColor=<rgba 0 0 0 1>}
  n92 = MaterialFunctionCall() {materialFunction=&GetUserInterfaceUV}
  n93 = ScalarParameter() {defaultValue=1, group=0 - Texture, parameterName=Global Scale}
  n94 = ScalarParameter() {group=0 - Texture, parameterName=Offset X}
  n95 = ScalarParameter() {group=0 - Texture, parameterName=Offset Y}
  NamedRerouteDeclaration(None=Add(
    A=MaterialFunctionCall(UVs=@n92.Normalized UV, Texture Scale=@n93) {materialFunction=&ScaleUVsByCenter}.UVs
    B=AppendVector(A=@n94, B=@n95)
   )) {name=Default UV, nodeColor=<rgba 0 0 0 1>}
 ]
}
