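# Translucent UI material for the Habanero lobby intro modal that draws a soft radial gradient flare with configurable two-color and opacity blends.
# An offset, center-scaled UI UV feeds a radial gradient mask, which is shaped by an ease-in sine curve.
# An optional static switch animates the flare: a delayed ping-pong timer drives a small threshold that is subtracted from the mask, which is then remapped, so the glow gently pulses.
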Material /Game/UI/Frontend/Lobby/Habanero/Habanero_Materials/M_UI_Habanero_IntroModal_FlareGradient
graph {
 properties {
  bUsedWithStaticMesh=false
  blendMode=BLEND_Translucent
  materialDomain=MD_UI
  pixelDepthOffsetMode=PDOM_Legacy
  usedShadingModels=Default Lit
 }
 nodes [
  LinearInterpolate(
   A=VectorParameter() {parameterName=Color B, defaultValue=<rgba 0 0 0 1>}.RGB
   B=VectorParameter() {parameterName=Color A, defaultValue=<rgba 1 1 1 1>}.RGB
   Alpha=NamedRerouteUsage()
  )
  LinearInterpolate(
   A=ScalarParameter() {parameterName=Opacity B, sliderMax=1}
   B=ScalarParameter() {defaultValue=1, parameterName=Opacity A, sliderMax=1}
   Alpha=NamedRerouteUsage()
  )
  n1 = Add(
   A=MaterialFunctionCall() {materialFunction=&GetUserInterfaceUV}.Normalized UV
   B=AppendVector(A=ScalarParameter() {defaultValue=-0.25, parameterName=OffsetX}, B=ScalarParameter() {defaultValue=0.5, parameterName=OffsetY})
  )
  n2 = ScalarParameter() {defaultValue=2, parameterName=Scale}
  NamedRerouteDeclaration(None=MaterialFunctionCall(UV=MaterialFunctionCall(UVs=@n1, Texture Scale=@n2) {materialFunction=&ScaleUVsByCenter}.UVs) {materialFunction=&MF_UI_RadialGradient}.Smooth) {name=Radial Gradient Mask, nodeColor=<rgba 0 0 0 1>}
  n3 = NamedRerouteUsage()
  n4 = ScalarParameter() {defaultValue=4, parameterName=Duration}
  n5 = ScalarParameter() {parameterName=Delay}
  n6 = LinearInterpolate(Alpha=MaterialFunctionCall!(
    Value=MaterialFunctionCall!(Animation Duration=@n4, Delay=@n5, Reset to Default) {materialFunction=/FortUILibrary/MaterialFunctions/Time/MF_UI_AnimationDelay}.Result
    Scale
    Height
   ) {materialFunction=/FortUILibrary/MaterialFunctions/Basics/MF_UI_PingPong}.Result) {constB=0.1}
  MaterialExpressionRemap_0 = MaterialXRemap!(None=Saturate(None=Subtract(A=@n3, B=@n6)), InputLow, InputHigh=OneMinus(None=@n6), TargetLow, TargetHigh)
  NamedRerouteDeclaration(None=MaterialFunctionCall("Value (0-1)"=StaticSwitchParameter(True=@MaterialExpressionRemap_0, False=@n3) {defaultValue=true, parameterName=Animated Flare}) {materialFunction=&MF_UI_EaseCurves}.EaseInSine) {name=Eased Radial Mask, nodeColor=<rgba 0 0 0 1>}
 ]
}
