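# Translucent UI material for the Habanero lobby's circular progress bar, with a main fill and a secondary delta fill.
# Circle SDFs with anti-aliasing build the ring, outline and inner disc, while radial arc functions mask Progress and Delta Progress.
# These masks layer parameterized colors and opacities into shared Color and Opacity reroutes, adding a ring glow and an eased edge gradient that blends between additive and override color.
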Material /Game/UI/Frontend/Lobby/Habanero/Habanero_Materials/M_UI_Habanero_ProgressBar
graph {
 properties {
  bUsedWithStaticMesh=false
  blendMode=BLEND_Translucent
  materialDomain=MD_UI
  pixelDepthOffsetMode=PDOM_Legacy
  usedShadingModels=Default Lit
 }
 nodes [
  n1 = NamedRerouteUsage()
  Ceil(None=@n1)
  n2 = NamedRerouteUsage()
  Ceil(None=@n2)
  n3 = MaterialFunctionCall(
   UVs=NamedRerouteUsage()
   Size=ScalarParameter() {defaultValue=0.55, parameterName=Progress Bar Size}
   OutlineThickness=ScalarParameter() {defaultValue=0.1, parameterName=Progress Bar Thickness}
  ) {materialFunction=&MF_UI_Circle}
  n4 = Constant() {r=0.02}
  n5 = ComponentMask(None=MaterialFunctionCall(DistanceField=NamedRerouteDeclaration(None=@n3.SDFOutline) {name=Bar Outline, nodeColor=<rgba 0 0 0 1>}, Threshold=@n4) {materialFunction=&MF_UI_AntiAliasing}.Inner) {r=true}
  n6 = NamedRerouteDeclaration(None=@n5) {name=Outer Ring Mask, nodeColor=<rgba 0 0 0 1>}
  Ceil(None=@n6)
  n7 = NamedRerouteDeclaration(None=@n5) {name=Progress Ring Mask, nodeColor=<rgba 0 0 0 1>}
  Ceil(None=@n7)
  Floor(None=@n6)
  NamedRerouteDeclaration(None=MaterialFunctionCall() {materialFunction=&GetUserInterfaceUV}.Normalized UV) {name=UV, nodeColor=<rgba 0 0 0 1>}
  n8 = ScalarParameter() {defaultValue=0.25, group=0 - Progress, parameterName=Progress, sliderMax=1, sortPriority=0}
  n9 = MaterialFunctionCall!(
   UV=NamedRerouteUsage()
   "Arc Start Angle (0-1)"
   "Arc End Angle (0-1)"
   "Fill Ratio (0-1)"=NamedRerouteDeclaration(None=OneMinus(None=@n8)) {name=Progress Value, nodeColor=<rgba 0 0 0 1>}
   Fill Clockwise=Constant() {r=1}
  ) {materialFunction=/FortUILibrary/MaterialFunctions/Basics/MF_UI_RadialArc}
  NamedRerouteDeclaration(None=Ceil(None=@n9.Radial Fill)) {name=Progress Mask, nodeColor=<rgba 0 0 0 1>}
  n10 = ScalarParameter() {defaultValue=0.05, parameterName=Bar Outline Distance}
  n11 = ScalarParameter() {defaultValue=0.1, parameterName=Progress Bar Thickness}
  n12 = Abs(None=MaterialFunctionCall(UVs=NamedRerouteUsage(), Size=ScalarParameter() {defaultValue=0.55, parameterName=Progress Bar Size}, OutlineThickness=Add(A=@n10, B=@n11)) {materialFunction=&MF_UI_Circle}.SDFOutline)
  n13 = Constant() {r=0.02}
  MaterialExpressionRemap_1 = MaterialXRemap!(None=NamedRerouteUsage(), InputLow=NamedRerouteUsage(), InputHigh, TargetLow, TargetHigh)
  n14 = OneMinus(None=Saturate(None=@MaterialExpressionRemap_1))
  NamedRerouteDeclaration(None=Multiply(
    A=MaterialFunctionCall(DistanceField=@n12, Threshold=@n13) {materialFunction=&MF_UI_AntiAliasing}.Inner
    B=@n14
   )) {name=Progress Bar Outline Mask, nodeColor=<rgba 0 0 0 1>}
  NamedRerouteDeclaration(None=@n9.Arc Mask) {name=Arc Mask, nodeColor=<rgba 0 0 0 1>}
  NamedRerouteDeclaration(None=MaterialFunctionCall("Value (0-1)"=@n14) {materialFunction=&MF_UI_EaseCurves}.EaseInQuint) {name=Gradient Edge Mask, nodeColor=<rgba 0 0 0 1>}
  n15 = NamedRerouteUsage()
  n16 = NamedRerouteDeclaration(None=OneMinus(None=ScalarParameter() {defaultValue=0.6, group=0 - Progress, parameterName=Delta Progress, sliderMax=1, sortPriority=1})) {name=Delta Value, nodeColor=<rgba 0 0 0 1>}
  n17 = Constant() {r=1}
  NamedRerouteDeclaration(None=Ceil(None=MaterialFunctionCall(UV=@n15, "Fill Ratio (0-1)"=@n16, Fill Clockwise=@n17) {materialFunction=&MF_UI_RadialArc}.Radial Fill)) {name=Delta Progress Mask, nodeColor=<rgba 0 0 0 1>}
  n18 = Multiply(
   A=Multiply(
    A=NamedRerouteUsage()
    B=ScalarParameter() {defaultValue=0.5, group=1 - Bar Visuals, parameterName=Delta Bar Opacity, sliderMax=1, sortPriority=11}
   ) {constB=0.5}
   B=@n7
  )
  n19 = Multiply(A=NamedRerouteUsage(), B=@n7)
  n20 = Multiply(A=@n6, B=ScalarParameter() {defaultValue=0.5, group=1 - Bar Visuals, parameterName=Progress Background Opacity, sliderMax=1, sortPriority=2})
  n21 = ScalarParameter() {defaultValue=0.75, parameterName=Inner Opacity, sliderMax=1}
  n22 = NamedRerouteUsage()
  n23 = ScalarParameter() {defaultValue=0.5, parameterName=Inner Background Size}
  n24 = ComponentMask(None=MaterialFunctionCall(
    DistanceField=MaterialFunctionCall(UVs=@n22, Size=@n23) {materialFunction=&MF_UI_Circle}.SDF
    Threshold=@n4
   ) {materialFunction=&MF_UI_AntiAliasing}.Inner) {g=true}
  n25 = Max(
   A=Max(
    A=NamedRerouteDeclaration(None=@n18) {name=Delta Bar Mask, nodeColor=<rgba 0 0 0 1>}
    B=NamedRerouteDeclaration(None=@n19) {name=Bar Mask, nodeColor=<rgba 0 0 0 1>}
   )
   B=LinearInterpolate(
    A=NamedRerouteDeclaration(None=@n20) {name=Progress Background Opacity, nodeColor=<rgba 0 0 0 1>}
    B=NamedRerouteDeclaration(None=@n21) {name=Inner Opacity, nodeColor=<rgba 0 0 0 1>}
    Alpha=NamedRerouteDeclaration(None=@n24) {name=Inner Circle Mask, nodeColor=<rgba 0 0 0 1>}
   )
  ) {constA=1}
  n26 = NamedRerouteUsage()
  n27 = Multiply(A=NamedRerouteUsage(), B=OneMinus(None=Floor(None=@n26)))
  NamedRerouteDeclaration(None=Max(A=@n25, B=@n27)) {name=Opacity, nodeColor=<rgba 0 0 0 1>}
  n28 = VectorParameter() {group=1 - Bar Visuals, parameterName=Progress Background Color, sortPriority=1, defaultValue=<rgba 0.4756 1 0.2723 1>}
  n29 = VectorParameter() {group=3 - Glow Properties, parameterName=Bar Glow Color, sortPriority=20, defaultValue=<rgba 1 0 0.5078 1>}
  n30 = Ceil(None=ScalarParameter() {defaultValue=1, group=3 - Glow Properties, parameterName=Glow Opacity, sliderMax=1, sortPriority=30})
  n31 = NamedRerouteUsage()
  n32 = Ceil(None=@n31)
  n33 = LinearInterpolate(
   A=LinearInterpolate(
    A=LinearInterpolate(A=@n28.RGB, B=@n29.RGB, Alpha=@n30)
    B=@n28.RGB
    Alpha=StaticSwitchParameter(True=@n32, False=@n31) {parameterName=Ceil Background Mask}
   )
   B=VectorParameter() {parameterName=Inner Color, defaultValue=<rgba 0.0331 0.7011 1 1>}.RGB
   Alpha=Multiply(A=NamedRerouteUsage(), B=NamedRerouteUsage())
  )
  n34 = VectorParameter() {group=1 - Bar Visuals, parameterName=Delta Bar Color, sortPriority=10, defaultValue=<rgba 0.2969 0.1561 0.0775 1>}
  n35 = NamedRerouteUsage()
  n36 = LinearInterpolate(
   A=LinearInterpolate(
    A=LinearInterpolate(A=@n33, B=@n34.RGB, Alpha=@n35)
    B=VectorParameter() {group=1 - Bar Visuals, parameterName=Bar Color, sortPriority=0, defaultValue=<rgba 1 1 1 1>}.RGB
    Alpha=@n1
   )
   B=VectorParameter() {parameterName=Bar Outline Color, defaultValue=<rgba 1 0.1253 0.0257 1>}.RGB
   Alpha=Min(A=@n1, B=NamedRerouteUsage())
  )
  n37 = Multiply(A=NamedRerouteUsage(), B=@n2)
  n38 = Add(A=@n36, B=@n37)
  n39 = VectorParameter() {parameterName=Edge Color, defaultValue=<rgba 0.1377 1 0.4096 1>}
  NamedRerouteDeclaration(None=LinearInterpolate(
    A=Saturate(None=@n38)
    B=LinearInterpolate(A=@n36, B=@n39.RGB, Alpha=@n37)
    Alpha=ScalarParameter() {parameterName=Override Edge Gradient, sliderMax=1}
   )) {name=Color, nodeColor=<rgba 0 0 0 1>}
  n40 = Saturate(None=OneMinus(None=NamedRerouteUsage()))
  n41 = ScalarParameter() {defaultValue=0.25, group=3 - Glow Properties, parameterName=Glow Range, sortPriority=0}
  n42 = OneMinus(None=@n41)
  MaterialExpressionRemap_0 = MaterialXRemap(None=Saturate(None=Subtract(A=@n40, B=@n42) {constB=0.75}), InputHigh=Saturate(None=@n41))
  n43 = ScalarParameter() {defaultValue=1, group=3 - Glow Properties, parameterName=Glow Opacity, sliderMax=1, sortPriority=30}
  NamedRerouteDeclaration(None=Saturate(None=Multiply(A=@MaterialExpressionRemap_0, B=@n43))) {name=Bar Glow, nodeColor=<rgba 0 0 0 1>}
  NamedRerouteUsage()
  NamedRerouteUsage()
  NamedRerouteUsage()
  Saturate(None=Add(A=@n25, B=@n27))
 ]
}
