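# Translucent UI material for the background behind Habanero progress text in the frontend lobby.
# A reflected gradient function is shaped by a sine ease-in curve, then multiplied by an Opacity parameter to form a soft fade mask.
# A separate Color parameter supplies a flat, adjustable tint, with no textures involved.
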Material /Game/UI/Frontend/Lobby/Habanero/Habanero_Materials/M_UI_Habanero_ProgressText_Bg
graph {
 properties {
  bUsedWithStaticMesh=false
  blendMode=BLEND_Translucent
  materialDomain=MD_UI
  pixelDepthOffsetMode=PDOM_Legacy
  usedShadingModels=Default Lit
 }
 nodes [
  Multiply(
   A=MaterialFunctionCall("Value (0-1)"=MaterialFunctionCall!(UVs, Offset, RotationCenter, RotationAngle) {materialFunction=/FortUILibrary/MaterialFunctions/Gradients/MF_UI_ReflectedGradient}.Result) {materialFunction=&MF_UI_EaseCurves}.EaseInSine
   B=ScalarParameter() {defaultValue=1, parameterName=Opacity, sliderMax=1}
  )
  VectorParameter() {parameterName=Color, defaultValue=<rgba 1 1 1 1>}
 ]
}
