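# Translucent UI material for a quest icon in the Habanero frontend lobby, with tintable icon, outline and radial glow.
# A flat icon texture supplies the icon mask in red and the outline mask in green, scaled by an opacity parameter.
# A rounded inverse outline mask hard-switches between outline and glow colors, then red blends in the icon color.
# That color multiplies the maximum of both masks and an opacity-scaled radial gradient.
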Material /Game/UI/Frontend/Lobby/Habanero/Habanero_Materials/M_UI_Habanero_QuestIcon
graph {
 properties {
  bUsedWithStaticMesh=false
  blendMode=BLEND_Translucent
  materialDomain=MD_UI
  pixelDepthOffsetMode=PDOM_Legacy
  usedShadingModels=Default Lit
 }
 nodes [
  n1 = VectorParameter() {parameterName=Outline Color, defaultValue=<rgba 0 1 0.0578 1>}
  n2 = VectorParameter() {parameterName=Glow Color, defaultValue=<rgba 0 1 0.0546 1>}
  n3 = TextureSample(
   UVs=MaterialFunctionCall() {materialFunction=&GetUserInterfaceUV}.Normalized UV
   Tex=TextureObject() {samplerType=SAMPLERTYPE_LinearColor, texture=/Game/UI/Frontend/Lobby/Habanero/Habanero_Textures/Flat/T_UI_Habanero_FlatIcons_08}
  )
  n4 = Multiply(A=@n3.G, B=ScalarParameter() {parameterName=Outline Opacity, sliderMax=1}) {constB=0}
  n5 = Round(None=OneMinus(None=@n4))
  n6 = MaterialFunctionCall() {materialFunction=&MF_UI_RadialGradient}
  n7 = ScalarParameter() {defaultValue=0.5, parameterName=Glow Opacity, sliderMax=1}
  Multiply(
   A=LinearInterpolate(
    A=LinearInterpolate(A=@n1.RGB, B=@n2.RGB, Alpha=@n5)
    B=VectorParameter() {parameterName=Icon Color, defaultValue=<rgba 1 1 1 1>}.RGB
    Alpha=@n3.R
   )
   B=Max(A=Max(A=@n3.R, B=@n4), B=Multiply(A=@n6.Smooth, B=@n7) {constB=0.5})
  )
 ]
}
