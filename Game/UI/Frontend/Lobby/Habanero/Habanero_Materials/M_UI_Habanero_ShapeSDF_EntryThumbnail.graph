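# UI material for Habanero lobby entry thumbnails that draws an SDF shape texture, with a squared vertical gradient between two colour parameters.
# The texture samples normalized UI UVs scaled about the centre, then an anti-aliasing function thresholds the distance field at its midpoint.
# A static switch adds floating motion by offsetting UVs vertically with time-driven ping-pong and sine easing between range parameters.
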Material /Game/UI/Frontend/Lobby/Habanero/Habanero_Materials/M_UI_Habanero_ShapeSDF_EntryThumbnail
graph {
 properties {
  bUsedWithStaticMesh=false
  blendMode=BLEND_Translucent
  materialDomain=MD_UI
  pixelDepthOffsetMode=PDOM_Legacy
  usedShadingModels=Default Lit
 }
 nodes [
  n1 = ComponentMask(None=TextureCoordinate()) {g=true}
  LinearInterpolate(
   A=VectorParameter() {parameterName=Color1, defaultValue=<rgba 1 0.7305 0 1>}.RGB
   B=VectorParameter() {parameterName=Color2, defaultValue=<rgba 0.4508 0.2159 0.004 1>}.RGB
   Alpha=Multiply(A=@n1, B=@n1)
  )
  n2 = MaterialFunctionCall(
   UVs=MaterialFunctionCall() {materialFunction=&GetUserInterfaceUV}.Normalized UV
   Texture Scale=ScalarParameter() {defaultValue=1, parameterName=Scale}
  ) {materialFunction=&ScaleUVsByCenter}
  n3 = Multiply(A=ScalarParameter() {defaultValue=1, parameterName=Range Min}) {constB=-1}
  n4 = ScalarParameter() {defaultValue=1, parameterName=Range Max}
  n5 = Multiply(A=Time(), B=ScalarParameter() {defaultValue=1, parameterName=Speed})
  n6 = MaterialFunctionCall("Value (0-1)"=MaterialFunctionCall!(Value=Frac(None=@n5), Scale, Height) {materialFunction=/FortUILibrary/MaterialFunctions/Basics/MF_UI_PingPong}.Result) {materialFunction=&MF_UI_EaseCurves}
  n7 = Add(A=@n2.UVs, B=AppendVector(A=Constant(), B=LinearInterpolate(A=@n3, B=@n4, Alpha=@n6.EaseInOutSine) {constA=-1}))
  MaterialFunctionCall(
   DistanceField=TextureSample(UVs=StaticSwitchParameter(True=@n7, False=@n2.UVs) {parameterName=Use Floating}, Tex=TextureObjectParameter() {parameterName=Texture}) {samplerType=SAMPLERTYPE_LinearGrayscale}.R
   Threshold=Constant() {r=0.5}
  ) {materialFunction=&MF_UI_AntiAliasing}
 ]
}
