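# Translucent UI material for the Habanero side panel background in the frontend lobby.
# A radial gradient on offset, center-scaled, fill-scaled UVs blends the base Color toward a brighter version of itself.
# The bright color is derived by doubling Color, raising it to the Radial Brightness exponent and saturating.
# A separate MF_UI_BoxAngle shape function, driven by Side and Angle parameters, defines an angled box shape.
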Material /Game/UI/Frontend/Lobby/Habanero/Habanero_Materials/M_UI_Habanero_SidePanel_Background
graph {
 properties {
  bUsedWithStaticMesh=false
  blendMode=BLEND_Translucent
  materialDomain=MD_UI
  pixelDepthOffsetMode=PDOM_Legacy
  usedShadingModels=Default Lit
 }
 nodes [
  n1 = VectorParameter() {parameterName=Color, defaultValue=<rgba 0.0024 0.016 0.117 1>}
  n2 = Add(A=@n1.RGB, B=@n1.RGB)
  n3 = ScalarParameter() {defaultValue=1, parameterName=Radial Brightness}
  n4 = ScalarParameter() {defaultValue=0.5, parameterName=Offset X}
  n5 = ScalarParameter() {defaultValue=0.5, parameterName=Offset Y}
  n6 = MaterialFunctionCall(
   UVs=Add(A=MaterialFunctionCall() {materialFunction=&GetUserInterfaceUV}.Normalized UV, B=AppendVector(A=@n4, B=@n5)) {constB=0.5}
   Texture Scale=ScalarParameter() {defaultValue=1, parameterName=Scale}
  ) {materialFunction=&ScaleUVsByCenter}
  LinearInterpolate(
   A=@n1.RGB
   B=Saturate(None=Power(Base=@n2, Exp=@n3) {constExponent=1})
   Alpha=MaterialFunctionCall(UV=MaterialFunctionCall(UV=@n6.UVs) {materialFunction=&MF_UI_UVScaling}.Scaled To Fill) {materialFunction=&MF_UI_RadialGradient}.Smooth
  )
  MaterialFunctionCall!(
   UV
   Corner=ScalarParameter() {defaultValue=1, parameterName=Side, sliderMax=1}
   Angle=ScalarParameter() {defaultValue=5, parameterName=Angle}
   Flip
   Falloff
  ) {materialFunction=/FortUILibrary/MaterialFunctions/Shapes/MF_UI_BoxAngle}
 ]
}
