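# Material function for Habanero frontend lobby materials that turns an SDF input into repeating, animated wave bands.
# It subtracts a Motion value from the SDF and multiplies by Repetitions, with Motion defaulting to a looping fraction of scaled time.
# It outputs that phase three ways: a cosine remapped to 0-1 for smooth waves, a Frac for sawtooth bands, and the raw value.
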MaterialFunction /Game/UI/Frontend/Lobby/Habanero/Habanero_Materials/MaterialFunctions/MF_Habanero_Wave
graph {nodes [
  n1 = Time()
  n2 = Constant() {desc=speed, r=0.5}
  n3 = FunctionInput(Preview=Frac(None=Multiply(A=@n1, B=@n2) {constB=0.5})) {
   bUsePreviewValueAsDefault=true
   description="0-1 value, default is time + frac"
   inputName=Motion
   inputType=FunctionInput_Scalar
   sortPriority=2
  }
  n4 = Multiply(
   A=Subtract(A=FunctionInput() {inputName=SDF Input, inputType=FunctionInput_Scalar}, B=@n3)
   B=FunctionInput(Preview=Constant() {desc=repetitions, r=3}) {bUsePreviewValueAsDefault=true, inputName=Repetitions, inputType=FunctionInput_Scalar, sortPriority=1}
  )
  n5 = Cosine(None=@n4)
  FunctionOutput(None=Add(A=Multiply(A=@n5) {constB=0.5, desc="cosine range is -1 to 1, so we need to bring it back to 0 to 1"}) {constB=0.5}) {outputName=Cosine}
  FunctionOutput(None=Frac(None=@n4)) {outputName=Frac, sortPriority=1}
  FunctionOutput(None=@n4) {outputName=Raw Output, sortPriority=2}
 ]}
