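# Translucent UI material for the Habanero lobby that draws a rank emblem with animated gems, plus a Desaturate control.
# An aspect-corrected RankTexture combines with pulsing radial background gradients and an optional wave-animated octagon pattern.
# Four crystals are sampled from a 2x2 gem atlas with mirrored, floating UVs, with channels packing glow, lighting, color and silhouette maps.
# Time-driven shine, moving glints and glare add highlights, and named reroutes link the stages.
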Material /Game/UI/Frontend/Lobby/Habanero/Habanero_Materials/M_UI_Habanero_WithGems
graph {
 properties {
  bAutomaticallySetUsageInEditor=false
  bUsedWithStaticMesh=false
  blendMode=BLEND_Translucent
  materialDomain=MD_UI
  usedShadingModels=Default Lit
 }
 nodes [
  Composite() {subgraphName=Fix widget aspect ratio}
  n1 = NamedRerouteUsage()
  n2 = LinearInterpolate(
   A=LinearInterpolate(A=NamedRerouteUsage(), B=NamedRerouteUsage(), Alpha=NamedRerouteUsage())
   B=NamedRerouteUsage()
   Alpha=NamedRerouteUsage()
  )
  n3 = Add(A=LinearInterpolate(B=@n2, Alpha=Multiply(A=@n1, B=@n2)), B=@n2)
  LinearInterpolate(A=@n3, B=Desaturation(None=@n3), Alpha=ScalarParameter() {parameterName=Desaturate})
  n4 = NamedRerouteUsage()
  n5 = NamedRerouteUsage()
  n6 = Constant2Vector() {g=1, r=1}
  n7 = MaterialFunctionCall!(
   UVs=Add(
    A=MaterialFunctionCall(UVs=@n5, Texture Scale=@n6) {materialFunction=&ScaleUVsByCenter}.UVs
    B=Constant2Vector() {g=-0.002}
   )
   Size=ScalarParameter() {defaultValue=0.5672, group=Pattern, parameterName=InnerPatternSize}
   Sides=Constant() {r=8}
   InnerAngle=Constant() {r=0.68}
   OutlineThickness=Constant() {r=0.02}
   OutlineAlignment
   GlowMin
   GlowMax
  ) {materialFunction=/FortUILibrary/MaterialFunctions/Shapes/MF_UI_RegularPolygon}
  n8 = Time()
  n9 = MaterialFunctionCall!(SDF Input=@n7.SDFOutline, Repetitions=Constant() {r=15}, Motion=OneMinus(None=Multiply(A=@n8) {constB=0.03})) {materialFunction=/Game/UI/Frontend/Lobby/Habanero/Habanero_Materials/MaterialFunctions/MF_Habanero_Wave}
  n10 = MaterialFunctionCall() {materialFunction=&LinearGradient}
  n11 = Divide(A=Power(Base=OneMinus(None=@n10.VGradient)))
  n12 = MaterialFunctionCall(
   A=Constant()
   B=Constant3Vector() {constant=<rgba 0.8796 0.008 0 1>}
   C=Constant() {r=1}
   Alpha=SmoothStep(Value=Subtract(A=@n9.Cosine, B=@n11)) {constMin=0.1216}
  ) {materialFunction=&Lerp_3Color}
  n13 = Constant() {r=0.3}
  n14 = Constant() {r=3}
  n15 = Multiply(
   A=@n12.Result
   B=OneMinus(None=MaterialFunctionCall(Radius=@n13, Density=@n14) {materialFunction=&RadialGradientExponential}.RadialGradientExponential)
  )
  NamedRerouteDeclaration(None=Multiply(A=LinearInterpolate(A=@n4, B=@n12.Result, Alpha=@n15), B=NamedRerouteUsage())) {name=InnerPattern_Color, nodeColor=<rgba 0.7813 0 1 1>}
  n16 = MaterialXAppend4Vector!(A=NamedRerouteUsage(), B=NamedRerouteUsage(), C=NamedRerouteUsage(), D=NamedRerouteUsage())
  n17 = ComponentMask(None=@n16) {b=true, g=true, r=true}
  n18 = LinearInterpolate(A=@n17, B=Multiply(A=@n17) {constB=1.6}, Alpha=NamedRerouteUsage())
  n19 = Multiply(A=@n18) {constB=10}
  n20 = Multiply(A=NamedRerouteUsage(), B=@n18)
  NamedRerouteDeclaration(None=Add(A=LinearInterpolate(B=@n19, Alpha=@n20), B=@n18)) {name=FinalGemColor, nodeColor=<rgba 0 0.0625 1 1>}
  n21 = MaterialFunctionCall(UVs=NamedRerouteUsage(), Texture Scale=Constant() {r=0.6}) {materialFunction=&ScaleUVsByCenter}
  NamedRerouteDeclaration(None=MaterialFunctionCall(UVs=@n21.UVs, CenterPosition=Constant2Vector(), Radius=Constant() {r=0.8}) {materialFunction=&RadialGradientExponential}.RadialGradientExponential) {name=Background_RadialGradient_Top, nodeColor=<rgba 0 1 0.1797 1>}
  NamedRerouteDeclaration(None=MaterialFunctionCall(UVs=@n21.UVs, CenterPosition=Constant2Vector() {g=1, r=0.5}, Radius=Constant() {r=0.6}, Density=Constant() {r=1.5}) {materialFunction=&RadialGradientExponential}.RadialGradientExponential) {name=Background_RadialGradient_Bottom, nodeColor=<rgba 0 0.3906 1 1>}
  n22 = Time()
  n23 = Constant() {r=0.2}
  n24 = Add(A=Frac(None=Multiply(A=@n22, B=@n23)), B=Constant())
  NamedRerouteDeclaration(None=LinearInterpolate(
    A=Constant() {r=0.2}
    B=Constant() {r=1}
    Alpha=MaterialFunctionCall(Value=@n24) {materialFunction=&LinearSine}.Linear Sine
   )) {name=BackgroundPulse_RadialGradient_Top, nodeColor=<rgba 0.7109 1 0 1>}
  n25 = Time()
  n26 = Constant() {r=0.15}
  n27 = Add(A=Frac(None=Multiply(A=@n25, B=@n26)), B=Constant() {r=0.1})
  NamedRerouteDeclaration(None=LinearInterpolate(
    A=Constant() {r=0.3}
    B=Constant() {r=0.8}
    Alpha=MaterialFunctionCall(Value=@n27) {materialFunction=&LinearSine}.Linear Sine
   )) {name=BackgroundPulse_RadialGradient_Bottom, nodeColor=<rgba 0.9688 0 1 1>}
  n28 = Constant3Vector() {constant=<rgba 0.0091 0.0009 0.0137 1>}
  n29 = Constant3Vector() {constant=<rgba 0.0732 0.0064 0.1698 1>}
  n30 = Multiply(A=NamedRerouteUsage(), B=NamedRerouteUsage())
  n31 = Constant3Vector() {constant=<rgba 0.474 0.041 0.0071 1>}
  n32 = Multiply(A=NamedRerouteUsage(), B=NamedRerouteUsage())
  NamedRerouteDeclaration(None=Max(A=LinearInterpolate(A=@n28, B=@n29, Alpha=@n30) {constB=0}, B=LinearInterpolate(A=@n28, B=@n31, Alpha=@n32))) {name=BackgroundGradient, nodeColor=<rgba 0 1 0.6484 1>}
  n33 = NamedRerouteUsage()
  n34 = NamedRerouteUsage()
  NamedRerouteDeclaration(None=Max(A=Max(A=@n33, B=@n34), B=NamedRerouteUsage())) {name=Opacity, nodeColor=<rgba 0.2422 0 1 1>}
  n35 = NamedRerouteUsage()
  n36 = ComponentMask(None=@n35) {g=true}
  n37 = LinearInterpolate(A=Add(A=@n36) {constB=0.03}, B=@n36, Alpha=NamedRerouteUsage())
  n38 = MaterialFunctionCall(UVs=AppendVector(A=ComponentMask(None=@n35) {r=true}, B=Add(A=@n37) {constB=-0.01}), "Rotation Angle (0-1)"=Constant() {r=0.25}) {materialFunction=&CustomRotator}
  n39 = SmoothStep(Value=Cosine(None=Add(A=@n38.Rotated Values) {constB=0.5})) {constMax=1.01, constMin=0.9}
  n40 = NamedRerouteUsage()
  n41 = NamedRerouteUsage()
  NamedRerouteDeclaration(None=LinearInterpolate(B=ComponentMask(None=@n39) {r=true}, Alpha=Max(A=@n40, B=@n41))) {name=GemMovingBlick, nodeColor=<rgba 1 0.4453 0 1>}
  NamedRerouteDeclaration(None=TextureObjectParameter() {group=0 - Texture, parameterName=Gem Texture, sortPriority=0}) {name=Gem Texture, nodeColor=<rgba 0 0 0 1>}
  n42 = NamedRerouteUsage()
  n43 = NamedRerouteUsage()
  n44 = NamedRerouteUsage()
  n45 = LinearInterpolate(
   A=LinearInterpolate(A=LinearInterpolate(A=@n42, B=@n43, Alpha=@n44), B=NamedRerouteUsage(), Alpha=NamedRerouteUsage())
   B=NamedRerouteUsage()
   Alpha=NamedRerouteUsage()
  )
  NamedRerouteDeclaration(None=ComponentMask(None=@n45) {r=true}) {name=Tier 3 Gems Glow Map, nodeColor=<rgba 0 0 0 1>}
  NamedRerouteDeclaration(None=ComponentMask(None=@n45) {b=true}) {name=Tier 3 Gems Color Map, nodeColor=<rgba 0 0 0 1>}
  n46 = Frac(None=Multiply(A=Time()) {constB=0.25})
  n47 = Min(A=@n46, B=OneMinus(None=@n46))
  NamedRerouteDeclaration(None=MaterialFunctionCall("Value (0-1)"=Multiply(A=@n47) {constB=2}) {materialFunction=&MF_UI_EaseCurves}.EaseInOutSine) {name=Floating Time, nodeColor=<rgba 0 0 0 1>}
  n48 = NamedRerouteUsage()
  NamedRerouteDeclaration(None=Multiply(A=ComponentMask(None=@n48) {a=true}, B=ScalarParameter() {defaultValue=1, group=Gems, parameterName=Gems Opacity, sliderMax=1})) {name=Gem Silhouette, nodeColor=<rgba 0 0 0 1>}
  NamedRerouteDeclaration(None=ComponentMask(None=@n45) {g=true}) {name=Tier 3 Gems Lighting Map, nodeColor=<rgba 0 0 0 1>}
  n49 = Max(A=NamedRerouteUsage(), B=NamedRerouteUsage())
  n50 = NamedRerouteUsage()
  NamedRerouteDeclaration(None=Max(A=Max(A=@n49, B=@n50), B=NamedRerouteUsage())) {name=Tier 3 Gems Silhouette, nodeColor=<rgba 0 0 0 1>}
  n51 = NamedRerouteUsage()
  n52 = Divide(A=Subtract(A=Multiply(A=@n51) {constB=0.5}) {constB=0.25}, B=Constant() {r=0.5})
  n53 = LinearInterpolate(Alpha=NamedRerouteUsage()) {constA=-0.005, constB=0.005}
  n54 = Add(A=Add(A=Add(A=@n52) {constB=0.25}, B=Constant2Vector()) {constB=2}, B=AppendVector(A=Constant(), B=Add(A=@n53) {constB=0.05}))
  n55 = NamedRerouteUsage()
  n56 = Multiply(
   A=MaterialFunctionCall!(UV=NamedRerouteUsage(), Grid Size=Constant() {r=2}, Index=Constant2Vector() {g=1, r=1}) {materialFunction=/FortUILibrary/MaterialFunctions/Basics/MF_UI_IsolateUV}.Result
   B=TextureSample(UVs=NamedRerouteDeclaration(None=@n54) {name=Gem 01 UV, nodeColor=<rgba 0 0 0 1>}, Tex=@n55) {samplerSource=SSM_Clamp_WorldGroupSettings, samplerType=SAMPLERTYPE_LinearColor}.RGBA
  )
  NamedRerouteDeclaration(None=ComponentMask(None=@n56) {a=true}) {name=Crystal 01 Mask, nodeColor=<rgba 0 0 0 1>}
  n57 = NamedRerouteUsage()
  n58 = Multiply(A=MaterialFunctionCall!(UVs=@n57, Alpha) {materialFunction=/FortUILibrary/MaterialFunctions/Basics/MF_UI_ClipEdgePixel}.Result, B=@n57)
  n59 = MaterialFunctionCall!(UV=@n58, Offset=Constant2Vector() {g=-0.05, r=-0.2}, Rotation, Scale=Constant2Vector() {g=1, r=-2}, MirrorY) {materialFunction=/FortUILibrary/MaterialFunctions/Basics/MF_UI_Symmetry_2x1_DEPRECATED}
  n60 = Divide(A=Subtract(A=Multiply(A=@n59.Result) {constB=0.5}) {constB=0.25}, B=Constant() {r=0.5})
  n61 = NamedRerouteUsage()
  n62 = NamedRerouteDeclaration(None=Round(None=ComponentMask(None=@n61) {r=true})) {name=Split UV, nodeColor=<rgba 0 0 0 1>}
  n63 = LinearInterpolate(
   A=LinearInterpolate(Alpha=@n62) {constA=-0.04, constB=0.04}
   B=LinearInterpolate(Alpha=@n62) {constA=-0.02, constB=0.02}
   Alpha=NamedRerouteUsage()
  ) {constA=-0.02, constB=0.02}
  n64 = Add(A=Add(A=Add(A=@n60) {constB=0.25}, B=Constant2Vector() {r=0.5}) {constB=2}, B=AppendVector(A=Constant(), B=Add(A=@n63) {constB=0.05}))
  n65 = Multiply(
   A=MaterialFunctionCall(UV=NamedRerouteUsage(), Grid Size=Constant() {r=2}, Index=Constant2Vector() {g=1, r=2}) {materialFunction=&MF_UI_IsolateUV}.Result
   B=TextureSample(UVs=NamedRerouteDeclaration(None=@n64) {name=Gem 02 UV, nodeColor=<rgba 0 0 0 1>}, Tex=@n55) {samplerSource=SSM_Clamp_WorldGroupSettings, samplerType=SAMPLERTYPE_LinearColor}.RGBA
  )
  NamedRerouteDeclaration(None=ComponentMask(None=@n65) {a=true}) {name=Crystal 02 Mask, nodeColor=<rgba 0 0 0 1>}
  n66 = MaterialFunctionCall(UV=@n58, Offset=Constant2Vector() {g=-0.05}, Scale=Constant2Vector() {g=1, r=-2}) {materialFunction=&MF_UI_Symmetry_2x1_DEPRECATED}
  n67 = Divide(A=Subtract(A=Multiply(A=@n66.Result) {constB=0.5}) {constB=0.25}, B=Constant() {r=0.5})
  n68 = Constant2Vector() {g=0.02, r=-0.01}
  n69 = Constant2Vector() {g=0.01, r=-0.02}
  n70 = NamedRerouteUsage()
  n71 = LinearInterpolate(
   A=LinearInterpolate(A=@n68.R, B=@n69.R, Alpha=@n70) {constB=0.01}
   B=LinearInterpolate(A=@n68.G, B=@n69.G, Alpha=@n70) {constA=-0.01, constB=0.1}
   Alpha=NamedRerouteUsage()
  ) {constA=-0.01, constB=0.01}
  n72 = Add(A=Add(A=Add(A=@n67) {constB=0.25}, B=Constant2Vector() {g=0.5}) {constB=2}, B=AppendVector(A=Constant(), B=Add(A=@n71) {constB=0.05}))
  n73 = Multiply(
   A=MaterialFunctionCall(UV=NamedRerouteUsage(), Grid Size=Constant() {r=2}, Index=Constant2Vector() {g=2, r=1}) {materialFunction=&MF_UI_IsolateUV}.Result
   B=TextureSample(UVs=NamedRerouteDeclaration(None=@n72) {name=Gem 03 UV, nodeColor=<rgba 0 0 0 1>}, Tex=@n55) {samplerSource=SSM_Clamp_WorldGroupSettings, samplerType=SAMPLERTYPE_LinearColor}.RGBA
  )
  NamedRerouteDeclaration(None=ComponentMask(None=@n73) {a=true}) {name=Crystal 03 Mask, nodeColor=<rgba 0 0 0 1>}
  n74 = MaterialFunctionCall(UV=@n58, Offset=Constant2Vector() {g=-0.1, r=0.15}, Scale=Constant2Vector() {g=1, r=-2}) {materialFunction=&MF_UI_Symmetry_2x1_DEPRECATED}
  n75 = Divide(A=Subtract(A=Multiply(A=@n74.Result) {constB=0.5}) {constB=0.25}, B=Constant() {r=0.5})
  n76 = NamedRerouteUsage()
  n77 = LinearInterpolate(
   A=LinearInterpolate(Alpha=@n76) {constA=-0.01, constB=0.03}
   B=LinearInterpolate(Alpha=@n76) {constA=-0.02, constB=0.02}
   Alpha=NamedRerouteUsage()
  ) {constA=-0.02, constB=0.02}
  n78 = Add(A=Add(A=Add(A=@n75) {constB=0.25}, B=Constant2Vector() {g=0.5, r=0.5}) {constB=2}, B=AppendVector(A=Constant(), B=Add(A=@n77) {constB=0.05}))
  n79 = Multiply(
   A=MaterialFunctionCall(UV=NamedRerouteUsage(), Grid Size=Constant() {r=2}, Index=Constant2Vector() {g=2, r=2}) {materialFunction=&MF_UI_IsolateUV}.Result
   B=TextureSample(UVs=NamedRerouteDeclaration(None=@n78) {name=Gem 04 UV, nodeColor=<rgba 0 0 0 1>}, Tex=@n55) {samplerSource=SSM_Clamp_WorldGroupSettings, samplerType=SAMPLERTYPE_LinearColor}.RGBA
  )
  NamedRerouteDeclaration(None=ComponentMask(None=@n79) {a=true}) {name=Crystal 04 Mask, nodeColor=<rgba 0 0 0 1>}
  NamedRerouteDeclaration(None=ComponentMask(None=@n79) {b=true, g=true, r=true}) {name=Crystal 04 Maps, nodeColor=<rgba 0 0 0 1>}
  NamedRerouteDeclaration(None=ComponentMask(None=@n73) {b=true, g=true, r=true}) {name=Crystal 03 Maps, nodeColor=<rgba 0 0 0 1>}
  NamedRerouteDeclaration(None=ComponentMask(None=@n56) {b=true, g=true, r=true}) {name=Crystal 01 Maps, nodeColor=<rgba 0 0 0 1>}
  NamedRerouteDeclaration(None=ComponentMask(None=@n65) {b=true, g=true, r=true}) {name=Crystal 02 Maps, nodeColor=<rgba 0 0 0 1>}
  NamedRerouteDeclaration(None=MaterialFunctionCall(UVs=NamedRerouteUsage(), Texture Scale=ScalarParameter() {defaultValue=1, group=Gems, parameterName=Gem Scale}) {materialFunction=&ScaleUVsByCenter}.UVs) {name=Default Gem UV, nodeColor=<rgba 0 0 0 1>}
  NamedRerouteDeclaration(None=@n16) {name=GemColor, nodeColor=<rgba 1 0 0.2812 1>}
  NamedRerouteDeclaration(None=NamedRerouteUsage()) {name=GemAlpha, nodeColor=<rgba 0 0.6016 1 1>}
  n80 = Multiply(A=MaterialFunctionCall(UVs=NamedRerouteUsage(), "Rotation Angle (0-1)"=Constant() {r=-0.125}) {materialFunction=&CustomRotator}.Rotated Values) {constB=0.5}
  n81 = MaterialFunctionCall("Value (0-1)"=MaterialFunctionCall!(Animation Duration=Constant() {r=2}, Delay=Constant() {r=2}, Reset to Default) {materialFunction=/FortUILibrary/MaterialFunctions/Time/MF_UI_AnimationDelay}.Result) {materialFunction=&MF_UI_EaseCurves}
  n82 = Frac(None=Subtract(A=Add(A=@n80), B=Subtract(A=@n81.EaseOutCubic) {constB=0.5}))
  n83 = Constant() {r=0.1}
  n84 = MaterialXRemap!(None=Subtract(A=Sine(None=@n82), B=OneMinus(None=@n83)) {constB=0.5}, InputLow, InputHigh=@n83, TargetLow, TargetHigh)
  NamedRerouteDeclaration(None=ComponentMask(None=Saturate(None=@n84)) {r=true}) {name=Shining Effect, nodeColor=<rgba 0 0 0 1>}
  n85 = AppendVector(
   A=ScalarParameter() {defaultValue=0.5, group=0 - Texture, parameterName=Anchor X}
   B=ScalarParameter() {defaultValue=0.5, group=0 - Texture, parameterName=Anchor Y}
  )
  n86 = Subtract(A=TextureCoordinate(), B=@n85)
  n87 = MaterialFunctionCall() {materialFunction=&GetUserInterfaceUV}
  n88 = AppendVector(A=Constant() {r=1}, B=Divide(A=ComponentMask(None=@n87.Pixel Size) {g=true}, B=ComponentMask(None=@n87.Pixel Size) {r=true})) {desc=Widget Vertical Aspect Ratio}
  MaterialExpressionReroute_22 = Reroute(None=@n88)
  n89 = Constant() {r=1}
  n90 = TextureObjectParameter() {group=0 - Texture, parameterName=RankTexture}
  n91 = TextureProperty(Tex=@n90) {property=TMTM_TexelSize}
  n92 = Divide(A=ComponentMask(None=@n91) {r=true}, B=ComponentMask(None=@n91) {g=true})
  n93 = Divide(
   A=Divide(A=@MaterialExpressionReroute_22.Output, B=AppendVector(A=@n89, B=@n92) {desc=Texture Vertical Aspect Ratio})
   B=ScalarParameter() {defaultValue=1, group=0 - Texture, parameterName=UserScale}
  )
  NamedRerouteDeclaration(None=Add(A=Multiply(A=@n86, B=@n93) {desc="Scaling to fix Aspect Ratio, fit to width"}, B=@n85)) {name=AdjustedTextureUV, nodeColor=<rgba 0.8984 1 0 1>}
  n94 = TextureSample(UVs=NamedRerouteUsage(), Tex=@n90) {samplerSource=SSM_Clamp_WorldGroupSettings}
  NamedRerouteDeclaration(None=Multiply(A=@n94.RGB, B=@n94.RGB)) {name=RankTexture, nodeColor=<rgba 1 0.6094 0 1>}
  NamedRerouteDeclaration(None=@n94.A) {name=TextureAlpha, nodeColor=<rgba 1 0.8672 0 1>}
  n95 = Add(
   A=NamedRerouteUsage()
   B=AppendVector(
    A=ScalarParameter() {group=Glare, parameterName=Glare_Offset_X}
    B=ScalarParameter() {defaultValue=-0.1, group=Glare, parameterName=Glare_Offset_Y}
   )
  )
  n96 = ScalarParameter() {defaultValue=0.1, group=Glare, parameterName=GlareRadius}
  n97 = ScalarParameter() {defaultValue=0.4, group=Glare, parameterName=GlareDensity}
  NamedRerouteDeclaration(None=Multiply(
    A=MaterialFunctionCall(UVs=@n95, Radius=@n96, Density=@n97) {materialFunction=&RadialGradientExponential}.RadialGradientExponential
    B=NamedRerouteUsage()
   )) {name=Lighting Spin, nodeColor=<rgba 0 0 0 1>}
  NamedRerouteDeclaration(None=Multiply(A=ScalarParameter() {parameterName=UseInnerPattern}, B=@n7.Fill)) {name=InnerAlpha, nodeColor=<rgba 0 1 0.4844 1>}
  NamedRerouteUsage()
  PinBase()
  PinBase!(Output=@n88) {reroutePins [Output {expression=&self:MaterialExpressionReroute_22}]}
 ]
}
