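# Translucent UI material for the Nexus lobby levelling progress bar, drawn as a circular ring that fills with a Progress parameter.
# Circle SDFs build the background and progress rings, and MF_UI_RadialArc masks the fill.
# Two circle SDFs, one rotated by progress, cap the arc's start and end.
# Feathered glow masks drive opacity and color blending, with hue shift and power adding variation along the glow.
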Material /Game/UI/Frontend/Lobby/Habanero/Habanero_Materials/M_UI_Nexus_LevellingProgressBar
graph {
 properties {
  bUsedWithStaticMesh=false
  blendMode=BLEND_Translucent
  materialDomain=MD_UI
  pixelDepthOffsetMode=PDOM_Legacy
  usedShadingModels=Default Lit
 }
 nodes [
  MaterialFunctionCall(Speed=Constant() {r=0.5}) {materialFunction=&DebugTimeSine}
  n1 = MaterialFunctionCall() {materialFunction=&GetUserInterfaceUV}
  NamedRerouteDeclaration(None=@n1.Normalized UV) {name=UV, nodeColor=<rgba 0 0 0 1>}
  n2 = ScalarParameter() {defaultValue=0.1, parameterName=Glow Feathering, sliderMax=1}
  n3 = Multiply(A=@n2) {constB=-1}
  n4 = NamedRerouteUsage()
  n5 = OneMinus(None=Saturate(None=SmoothStep(Min=@n3, Max=@n2, Value=@n4)))
  n6 = Saturate(None=SmoothStep(Min=@n3, Max=@n2, Value=NamedRerouteUsage()))
  n7 = Multiply(A=LinearInterpolate(
    A=NamedRerouteDeclaration(None=@n5) {name=Start/End Points Stepped, nodeColor=<rgba 0 0 0 1>}
    B=OneMinus(None=@n6)
    Alpha=NamedRerouteUsage()
   )) {constB=2}
  n8 = ScalarParameter() {defaultValue=0.4, parameterName=Glow Opacity, sliderMax=1}
  n9 = Ceil(None=NamedRerouteUsage())
  NamedRerouteDeclaration(None=Multiply(A=Saturate(None=@n7), B=Multiply(A=@n8, B=@n9))) {name=Glow Mask, nodeColor=<rgba 0 0 0 1>}
  n10 = NamedRerouteDeclaration(None=ScalarParameter() {defaultValue=0.07, parameterName=Outline BG Thickness}) {name=Outline BG Thickness, nodeColor=<rgba 0 0 0 1>}
  n11 = MaterialFunctionCall(
   UVs=NamedRerouteUsage()
   Size=NamedRerouteDeclaration(None=ScalarParameter() {defaultValue=0.99, parameterName=Size}) {name=Size, nodeColor=<rgba 0 0 0 1>}
   OutlineThickness=@n10
  ) {materialFunction=&MF_UI_Circle}
  n12 = Abs(None=Add(A=NamedRerouteDeclaration(None=@n11.SDF) {name=Shape SDF, nodeColor=<rgba 0 0 0 1>}, B=@n10))
  n13 = ScalarParameter() {defaultValue=0.042, parameterName=Outline Progress Thickness, sliderMax=1}
  NamedRerouteDeclaration(None=Subtract(A=@n12, B=NamedRerouteDeclaration(None=@n13) {name=Outline Progress Thickness, nodeColor=<rgba 0 0 0 1>})) {name=Outline Progress SDF, nodeColor=<rgba 0 0 0 1>}
  n14 = Constant() {r=0.015}
  n15 = Multiply(
   A=NamedRerouteDeclaration(None=Multiply(A=@n14) {constB=0.5}) {name=Edge Falloff Value, nodeColor=<rgba 0 0 0 1>}
   B=Divide(B=NamedRerouteUsage()) {constA=250, constB=1}
  )
  n16 = Multiply(A=@n15) {constB=-1}
  n17 = SmoothStep(Min=@n16, Max=@n15, Value=NamedRerouteUsage()) {constMax=0}
  NamedRerouteDeclaration(None=OneMinus(None=Saturate(None=@n17))) {name=Progression LVL Mask, nodeColor=<rgba 0 0 0 1>}
  n18 = LinearInterpolate(
   A=LinearInterpolate(
    A=ScalarParameter() {defaultValue=0.2, parameterName=Background Opacity, sliderMax=1}
    B=ScalarParameter() {defaultValue=0.7, parameterName=Background Progress Opacity, sliderMax=1}
    Alpha=NamedRerouteUsage()
   )
   Alpha=Multiply(A=NamedRerouteUsage(), B=NamedRerouteUsage())
  )
  n19 = NamedRerouteUsage()
  NamedRerouteDeclaration(None=LinearInterpolate(A=Multiply(A=@n18, B=@n19), Alpha=NamedRerouteUsage())) {name=Opacity, nodeColor=<rgba 0 0 0 1>}
  NamedRerouteDeclaration(None=Subtract(A=@n12, B=@n10) {constB=0.1}) {name=Outline BG SDF, nodeColor=<rgba 0 0 0 1>}
  NamedRerouteDeclaration(None=Min(A=ComponentMask(None=@n1.Pixel Size) {r=true}, B=ComponentMask(None=@n1.Pixel Size) {g=true})) {name=Min Pixel Size, nodeColor=<rgba 0 0 0 1>}
  n20 = SmoothStep(Min=@n16, Max=@n15, Value=NamedRerouteUsage()) {constMax=0}
  NamedRerouteDeclaration(None=OneMinus(None=Saturate(None=@n20))) {name=Ring Mask, nodeColor=<rgba 0 0 0 1>}
  n21 = SmoothStep(Min=@n16, Max=@n15, Value=NamedRerouteUsage()) {constMax=0}
  NamedRerouteDeclaration(None=OneMinus(None=Saturate(None=@n21))) {name=Circle Mask, nodeColor=<rgba 0 0 0 1>}
  n22 = NamedRerouteUsage()
  n23 = OneMinus(None=NamedRerouteDeclaration(None=ScalarParameter() {parameterName=Progress, sliderMax=1}) {name=Progress Value, nodeColor=<rgba 0 0 0 1>})
  n24 = ScalarParameter() {defaultValue=1, parameterName=Fill Clockwise, sliderMax=1}
  n25 = Ceil(None=NamedRerouteDeclaration(None=MaterialFunctionCall!(UV=@n22, "Arc Start Angle (0-1)", "Arc End Angle (0-1)", "Fill Ratio (0-1)"=@n23, Fill Clockwise=@n24) {materialFunction=/FortUILibrary/MaterialFunctions/Basics/MF_UI_RadialArc}.Radial Fill) {name=Progress Mask Gradient, nodeColor=<rgba 0 0 0 1>})
  NamedRerouteDeclaration(None=LinearInterpolate(A=NamedRerouteUsage(), Alpha=NamedRerouteDeclaration(None=@n25) {name=Progress Mask, nodeColor=<rgba 0 0 0 1>})) {name=Progress Mask with points, nodeColor=<rgba 0 0 0 1>}
  n26 = Constant2Vector() {g=-0.5}
  n27 = NamedRerouteUsage()
  n28 = NamedRerouteUsage()
  n29 = AppendVector(A=@n26.R, B=Add(A=Multiply(A=@n26.G, B=@n27), B=Multiply(A=@n28) {constB=0.5}))
  n30 = NamedRerouteUsage()
  n31 = MaterialFunctionCall(UVs=Subtract(A=NamedRerouteUsage(), B=@n29), Size=@n30) {materialFunction=&MF_UI_Circle}
  n32 = NamedRerouteUsage()
  n33 = OneMinus(None=NamedRerouteUsage())
  n34 = MaterialFunctionCall(UVs=Subtract(A=Rotator(Coordinate=@n32, Time=@n33) {speed=6.2832}, B=@n29), Size=@n30) {materialFunction=&MF_UI_Circle}
  NamedRerouteDeclaration(None=MaterialFunctionCall!(
    DistanceFieldA=NamedRerouteDeclaration(None=@n31.SDF) {name=Start Point, nodeColor=<rgba 0 0 0 1>}
    DistanceFieldB=NamedRerouteDeclaration(None=@n34.SDF) {name=End Point, nodeColor=<rgba 0 0 0 1>}
    BlendAmount
   ) {materialFunction=/FortUILibrary/MaterialFunctions/Shapes/Modifiers/MF_UI_SDF_Union}.Exact) {name=Start/End Points, nodeColor=<rgba 0 0 0 1>}
  n35 = NamedRerouteUsage()
  n36 = Saturate(None=SmoothStep(Min=@n16, Max=@n15, Value=NamedRerouteUsage()) {constMax=0})
  NamedRerouteDeclaration(None=Multiply(A=Ceil(None=@n35), B=OneMinus(None=@n36))) {name=Points Mask, nodeColor=<rgba 0 0 0 1>}
  NamedRerouteUsage()
  n37 = ScalarParameter() {defaultValue=0.1, parameterName=Glow Feathering, sliderMax=1}
  n38 = Multiply(A=@n37) {constB=-1}
  n39 = NamedRerouteUsage()
  n40 = OneMinus(None=Saturate(None=SmoothStep(Min=@n38, Max=@n37, Value=@n39)))
  n41 = SmoothStep(Max=NamedRerouteUsage(), Value=NamedRerouteUsage())
  n42 = ScalarParameter() {defaultValue=0.1, parameterName=Glow Feathering, sliderMax=1}
  n43 = Saturate(None=SmoothStep(Min=Multiply(A=@n42) {constB=-1}, Max=@n42, Value=NamedRerouteUsage()))
  n44 = Saturate(None=Subtract(A=Max(A=@n40, B=@n41), B=OneMinus(None=@n43)))
  n45 = Multiply(A=@n44) {constB=-0.22}
  n46 = VectorParameter() {parameterName=Background Color, defaultValue=<rgba 0 0 0 1>}
  n47 = NamedRerouteUsage()
  n48 = NamedRerouteUsage()
  n49 = LinearInterpolate(
   A=LinearInterpolate(A=@n46.RGB, B=VectorParameter() {parameterName=Progress Background Color, defaultValue=<rgba 0 0 0 1>}.RGB, Alpha=NamedRerouteUsage())
   B=@n46.RGB
   Alpha=Saturate(None=Subtract(A=@n47, B=@n48))
  )
  n50 = VectorParameter() {parameterName=Glow Color, defaultValue=<rgba 0.1753 0 0.25 1>}
  n51 = OneMinus(None=NamedRerouteUsage())
  n52 = VectorParameter() {parameterName=Outside Glow Color, defaultValue=<rgba 0.1582 0.0549 0.4219 1>}
  n53 = NamedRerouteUsage()
  n54 = LinearInterpolate(
   A=LinearInterpolate(
    A=LinearInterpolate(A=@n49, B=@n50.RGB, Alpha=@n51)
    B=LinearInterpolate(A=@n52.RGB, B=@n50.RGB, Alpha=@n53)
    Alpha=MaterialFunctionCall("Value (0-1)"=@n53) {materialFunction=&MF_UI_EaseCurves}.EaseOutSine
   )
   B=VectorParameter() {parameterName=Progress Color 1, defaultValue=<rgba 0.7151 0.0469 1 1>}.RGB
   Alpha=Multiply(A=NamedRerouteUsage(), B=NamedRerouteUsage())
  )
  n55 = Add(A=@n44)
  Saturate(None=Power(
    Base=MaterialFunctionCall(Hue Shift Percentage=@n45, Texture=@n54) {materialFunction=&HueShift}.Result
    Exp=Multiply(A=@n55) {constB=0.75}
   ))
 ]
}
