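# Translucent UI material that draws an animated circular Power Rank badge with a lightning bolt icon for the Habanero frontend lobby.
# MSDF samples shape the bolt and its outline, layered with a gradient ring and glow, noise-distorted polar swirl, scrolling particles and shine.
# A custom HLSL loop repeats rotating radial gradients into animated outline lines.
# Combined masks index a curve-atlas color ramp composited with the background and form the opacity.
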Material /Game/UI/Frontend/Lobby/Habanero/Habanero_Materials/M_UI_PowerRank
graph {
 properties {bAutomaticallySetUsageInEditor=false, blendMode=BLEND_Translucent, materialDomain=MD_UI, usedShadingModels=Default Lit}
 nodes [
  n1 = ScalarParameter() {defaultValue=1, parameterName=Scale}
  NamedRerouteDeclaration(None=MaterialFunctionCall(
    UVs=MaterialFunctionCall() {materialFunction=&MF_UI_UVScaling}.Scaled to Fit
    Texture Scale=AppendVector(A=@n1, B=@n1)
   ) {materialFunction=&ScaleUVsByCenter}.UVs) {name=UV, nodeColor=<rgba 0 1 0.1328 1>}
  n2 = MaterialFunctionCall(GradientStart=Constant() {r=0.4}, GradientEnd=Constant() {r=0.6}, RotationAngle=Constant() {r=0.75}) {materialFunction=&MF_UI_LinearGradient}
  n3 = ScalarParameter() {defaultValue=1, group=Icon, parameterName=Bolt Icon Scale}
  n4 = Add(
   A=MaterialFunctionCall(UVs=NamedRerouteUsage(), Texture Scale=AppendVector(A=@n3, B=@n3)) {materialFunction=&ScaleUVsByCenter}.UVs
   B=Constant2Vector() {g=-0.01, r=0.03}
  )
  n5 = TextureSample(UVs=MaterialFunctionCall(UVs=@n4, Texture Scale=Constant2Vector() {g=0.6, r=0.5}) {materialFunction=&ScaleUVsByCenter}.UVs) {
   samplerSource=SSM_Clamp_WorldGroupSettings
   samplerType=SAMPLERTYPE_LinearColor
   texture=/Game/UI/Frontend/Lobby/Habanero/Habanero_Textures/T_UI_Habanero_PowerRank_MSDF
  }
  n6 = MaterialFunctionCall!(MSDF RGB=@n5.RGB) {materialFunction=/FortUILibrary/MaterialFunctions/Shapes/Modifiers/MF_UI_MSDFtoSDF}
  n7 = SmoothStep(Value=@n6.SDF) {constMax=0.57, constMin=0.51}
  n8 = TextureSample(UVs=MaterialFunctionCall(UVs=@n4, Texture Scale=Constant2Vector() {g=0.65, r=0.55}) {materialFunction=&ScaleUVsByCenter}.UVs) {
   samplerSource=SSM_Clamp_WorldGroupSettings
   samplerType=SAMPLERTYPE_LinearColor
   texture=&T_UI_Habanero_PowerRank_MSDF
  }
  n9 = SmoothStep(Value=MaterialFunctionCall(MSDF RGB=@n8.RGB) {materialFunction=&MF_UI_MSDFtoSDF}.SDF) {constMin=0.2}
  n10 = Subtract(A=Clamp(None=Multiply(A=@n9) {constB=2}) {maxDefault=0.8}, B=@n7)
  NamedRerouteDeclaration(None=LinearInterpolate(A=Multiply(A=@n10, B=@n2.Result), B=@n7, Alpha=@n7)) {name=LightningSymbol_Alpha, nodeColor=<rgba 1 0.0469 0 1>}
  NamedRerouteDeclaration(None=TextureObject() {desc=Rift Noise Texture, samplerType=SAMPLERTYPE_Grayscale, texture=/FortUILibrary/UIKit/Buttons/Textures/LowResBlurredNoise}) {name=Rift Noise Texture, nodeColor=<rgba 0 0 0 1>}
  n11 = NamedRerouteUsage()
  n12 = MaterialFunctionCall(DistanceField=NamedRerouteUsage(), Threshold=Constant(), Smoothing=Constant() {r=0.01}) {materialFunction=&MF_UI_AntiAliasing}
  n13 = Multiply(
   A=MaterialFunctionCall(
    UVs=NamedRerouteUsage()
    Radius=ScalarParameter() {defaultValue=0.422, group=Outline, parameterName=4 - Outline Glow Radius}
    Density=ScalarParameter() {defaultValue=10, group=Outline, parameterName=5 - Outline Glow Density}
   ) {materialFunction=&RadialGradientExponential}.RadialGradientExponential
   B=OneMinus(None=Max(A=@n11, B=@n12.Inner))
  )
  n14 = ScalarParameter() {group=Outline, parameterName=6 - Otline Glow Gradient Start}
  n15 = ScalarParameter() {defaultValue=0.7, group=Outline, parameterName=7 - Otline Glow Gradient End}
  n16 = Constant() {r=0.75}
  NamedRerouteDeclaration(None=Multiply(
    A=@n13
    B=MaterialFunctionCall(GradientStart=@n14, GradientEnd=@n15, RotationAngle=@n16) {materialFunction=&MF_UI_LinearGradient}.Result
   )) {name=OutlineGlow_Floating, nodeColor=<rgba 0 0.7422 1 1>}
  NamedRerouteDeclaration(None=@n13) {name=OutlineGlow, nodeColor=<rgba 1 0.9375 0 1>}
  n17 = NamedRerouteUsage()
  n18 = MaterialFunctionCall(Texture=NamedRerouteUsage(), UVControl_A=Constant() {r=0.3}, "[Override]TimeInput"=NamedRerouteUsage(), "[Override]UVInput"=@n17) {materialFunction=&FXCommon_TextureSample}
  n19 = MaterialFunctionCall!(NoiseValue=@n18.RG, Control_RGB=Constant3Vector() {constant=<rgba -0.5 -0.5 0 0>}, Control_A=Constant() {r=0.2}) {materialFunction=/Game/Effects/Niagara/MasterMaterials/Functions/Common/FXCommon_NoiseControl}
  NamedRerouteDeclaration(None=MaterialFunctionCall!(Vector or UVs=Add(A=@n19.NoiseValue, B=@n17)) {materialFunction=/FortUILibrary/MaterialFunctions/Basics/MF_UI_CheaperVectorToRadial}.Radial Coordinates) {name=FloatingPolarCoordinates, nodeColor=<rgba 0.5937 1 0 1>}
  NamedRerouteDeclaration(None=Multiply(A=VectorParameter() {parameterName=BackgroundColor, defaultValue=<rgba 0.0037 0 0.0048 1>}.RGB, B=NamedRerouteUsage())) {name=Background, nodeColor=<rgba 0.1953 0 1 1>}
  n20 = ScalarParameter() {defaultValue=0.4, group=Icon, parameterName=Bolt Icon Outline Whidth}
  n21 = Max(A=SmoothStep(Min=@n20, Max=Add(A=@n20) {constB=0.14}, Value=@n6.SDF) {constMax=0.47, constMin=0.4}, B=Divide(A=@n10) {constB=3})
  NamedRerouteDeclaration(None=Multiply(A=Subtract(A=@n21, B=@n7), B=NamedRerouteUsage())) {name=LightningSymbolOutline_Alpha, nodeColor=<rgba 0 0.2969 1 1>}
  NamedRerouteDeclaration(None=@n7) {name=LightIcon, nodeColor=<rgba 0.4531 0 1 1>}
  n22 = OneMinus(None=ScalarParameter() {defaultValue=0.7, group=Particles, parameterName=Particles Size})
  n23 = NamedRerouteUsage()
  n24 = ScalarParameter() {defaultValue=0.7, group=Particles, parameterName=ParticlesScale}
  n25 = AppendVector(A=@n24, B=@n24)
  n26 = Constant()
  n27 = Frac(None=Multiply(A=Time()) {constB=0.02})
  n28 = TextureSample(
   UVs=Add(
    A=MaterialFunctionCall(UVs=@n23, Texture Scale=@n25) {materialFunction=&ScaleUVsByCenter}.UVs
    B=AppendVector(A=@n26, B=@n27)
   )
   Tex=TextureObject() {samplerType=SAMPLERTYPE_Alpha, texture=/Game/UI/Frontend/Lobby/Habanero/Habanero_Textures/T_UI_Habanero_Clouds_Stars}
  )
  n29 = Saturate(None=Multiply(A=SmoothStep(Min=@n22, Value=@n28.R) {constMin=0.5}) {constB=4})
  n30 = NamedRerouteUsage()
  n31 = NamedRerouteUsage()
  n32 = Multiply(A=OneMinus(None=Max(A=@n30, B=@n31)), B=NamedRerouteUsage())
  NamedRerouteDeclaration(None=Multiply(A=Multiply(A=@n29, B=@n32), B=ScalarParameter() {defaultValue=1, group=Particles, parameterName=ParticlesOpacity})) {name=Particles, nodeColor=<rgba 0 0 0 1>}
  n33 = NamedRerouteUsage()
  n34 = MaterialFunctionCall(
   UVs=MaterialFunctionCall(UVs=NamedRerouteUsage(), Texture Scale=Constant2Vector() {g=1, r=1.05}) {materialFunction=&ScaleUVsByCenter}.UVs
   CenterPosition=Constant2Vector() {g=0.4416, r=0.5}
   Radius=ScalarParameter() {defaultValue=0.422, parameterName=Bottom Shining Radius}
   Density=Constant() {r=5}
  ) {materialFunction=&RadialGradientExponential}
  n35 = Multiply(A=OneMinus(None=NamedRerouteUsage()), B=OneMinus(None=Saturate(None=@n34.RadialGradientExponential)))
  NamedRerouteDeclaration(None=Saturate(None=Multiply(A=@n35, B=@n33))) {name=ButtonShineAlpha, nodeColor=<rgba 1 0.1875 0 1>}
  n36 = NamedRerouteUsage()
  n37 = NamedRerouteUsage()
  n38 = VectorParameter() {group=Swirl, parameterName=Swirl Noise Params, defaultValue=<rgba 1 -0.9 0.1 0.1>}
  n39 = Multiply(
   A=TextureSample!(
    UVs=MaterialFunctionCall!(UVInput=@n36, TimeInput=@n37, Control_RGB=@n38.RGB, Control_A=@n38.RGB) {materialFunction=/Game/Effects/Niagara/MasterMaterials/Functions/Common/FXCommon_QuickUVControl}.UV
    Tex=NamedRerouteUsage()
    Level
    Apply View MipBias
   ) {mipValueMode=TMVM_MipLevel, texture=/Game/UI/Frontend/Lobby/Habanero/Habanero_Textures/T_UI_HabaneroWaves}.RGB
   B=LinearInterpolate(A=NamedRerouteUsage(), B=NamedRerouteUsage(), Alpha=ScalarParameter() {defaultValue=1, parameterName=Use Radial Glow Instead of Noise})
  )
  n40 = Constant() {r=4}
  n41 = Multiply(
   A=Multiply(A=Multiply(A=@n39, B=@n40), B=NamedRerouteUsage())
   B=ScalarParameter() {defaultValue=1, group=Swirl, parameterName=Swirl Opacity}
  )
  NamedRerouteDeclaration(None=Saturate(None=Multiply(A=@n41) {constB=4})) {name=Swirl Mask, nodeColor=<rgba 0 0 0 1>}
  n42 = MaterialFunctionCall!(FloatValue, Speed=ScalarParameter() {defaultValue=0.05, group=Outline Lines, parameterName=Outline Lines Time Speed _ 1}) {materialFunction=/FortUILibrary/MaterialFunctions/Basics/MF_UI_LinearTime}
  n43 = MaterialFunctionCall(UVs=NamedRerouteUsage(), "Rotation Angle (0-1)"=Multiply(A=@n42.0-1 Loop) {constB=0.2}) {materialFunction=&CustomRotator}
  n44 = AppendVector(A=Constant() {r=0.7}, B=Add(A=Multiply(A=@n42.0-1 Ping pong) {constB=0.1}) {constB=0.9})
  n45 = MaterialFunctionCall(Speed=ScalarParameter() {defaultValue=0.07, group=Outline Lines, parameterName=Outline Lines Time Speed _ 2}) {materialFunction=&MF_UI_LinearTime}
  n46 = MaterialFunctionCall(UVs=NamedRerouteUsage(), "Rotation Angle (0-1)"=Multiply(A=@n45.0-1 Loop) {constB=0.7}) {materialFunction=&CustomRotator}
  n47 = AppendVector(A=Constant() {r=0.7}, B=Add(A=Multiply(A=@n45.0-1 Ping pong) {constB=0.1}) {constB=0.9})
  MaterialExpressionMax_17 = Max(
   A=MaterialFunctionCall(
    UVs=MaterialFunctionCall(UVs=@n43.Rotated Values, Texture Scale=@n44) {materialFunction=&ScaleUVsByCenter}.UVs
    Radius=Constant() {r=0.422}
    Density=Constant() {r=10}
   ) {materialFunction=&RadialGradientExponential}.RadialGradientExponential
   B=MaterialFunctionCall(
    UVs=MaterialFunctionCall(UVs=@n46.Rotated Values, Texture Scale=@n47) {materialFunction=&ScaleUVsByCenter}.UVs
    Radius=Constant() {r=0.422}
    Density=Constant() {r=10}
   ) {materialFunction=&RadialGradientExponential}.RadialGradientExponential
  )
  MaterialExpressionScalarParameter_27 = ScalarParameter() {defaultValue=10, group=Outline Lines, parameterName=Outline Lines Amount}
  MaterialExpressionScalarParameter_29 = ScalarParameter() {defaultValue=0.07, group=Outline Lines, parameterName=Outline Lines Thickness}
  MaterialExpressionConstant_34 = Constant() {r=0.04}
  MaterialExpressionScalarParameter_28 = ScalarParameter() {defaultValue=0.2, group=Outline Lines, parameterName=Outline Lines Spacing}
  MaterialExpressionConstant_36 = Constant()
  n48 = Custom!(
   sdf=@MaterialExpressionMax_17
   num=@MaterialExpressionScalarParameter_27
   thickness=@MaterialExpressionScalarParameter_29
   glow=@MaterialExpressionConstant_34
   spacing=@MaterialExpressionScalarParameter_28
   skew=@MaterialExpressionConstant_36
  ) {
   code="float output = 0.0;\r\nGlowOutput = 0.0;\r\nfor(int i = 0; i < num; i++) {\r\noutput += 1-smoothstep(0.0, (thickness+(skew*i)), abs(sdf-spacing*i));\r\nGlowOutput += 1-smoothstep(-.1,( glow+(skew*i)), abs(sdf-spacing*i));\r\n}\r\nreturn output;"
   description=RepeatSDFOutline
   outputType=CMOT_Float1
   additionalOutputs [{outputName=GlowOutput, outputType=CMOT_Float1}]
   inputs [
    {inputName=sdf, input {expression=&self:MaterialExpressionMax_17}}
    {inputName=num, input {expression=&self:MaterialExpressionScalarParameter_27}}
    {
     inputName=thickness
     input {expression=&self:MaterialExpressionScalarParameter_29}
    }
    {inputName=glow, input {expression=&self:MaterialExpressionConstant_34}}
    {
     inputName=spacing
     input {expression=&self:MaterialExpressionScalarParameter_28}
    }
    {inputName=skew, input {expression=&self:MaterialExpressionConstant_36}}
   ]
  }
  NamedRerouteDeclaration(None=Multiply(A=OneMinus(None=@n48.return), B=ScalarParameter() {defaultValue=1, group=Outline Lines, parameterName=Outline Lines Opacity})) {name=Lines, nodeColor=<rgba 1 0 0.3516 1>}
  n49 = MaterialFunctionCall(DistanceField=NamedRerouteUsage(), Threshold=Constant()) {materialFunction=&MF_UI_AntiAliasing}
  n50 = MaterialFunctionCall(
   UVs=NamedRerouteUsage()
   Size=Constant() {r=0.7}
   OutlineThickness=ScalarParameter() {defaultValue=0.03, group=Outline, parameterName=1- Outline Thickness}
  ) {materialFunction=&MF_UI_Circle}
  NamedRerouteDeclaration(None=Multiply(A=OneMinus(None=@n49.Inner), B=@n50.Fill)) {name=InnerAlpha, nodeColor=<rgba 1 0 0.1875 1>}
  NamedRerouteDeclaration(None=Divide(A=Time(), B=ScalarParameter() {defaultValue=20, group=Swirl, parameterName=Swirl Time}) {constB=10}) {name=InnerShapeTime, nodeColor=<rgba 0 0.2266 1 1>}
  n51 = Time()
  n52 = ScalarParameter() {defaultValue=-1, group=Swirl, parameterName=Swirl Radial Falloff Mask Time}
  n53 = MaterialFunctionCall(Vector or UVs=Rotator(Coordinate=NamedRerouteUsage(), Time=Multiply(A=@n51, B=@n52) {constB=-1}) {speed=1}) {materialFunction=&MF_UI_CheaperVectorToRadial}
  n54 = SmoothStep(Min=ScalarParameter() {defaultValue=0.95, group=Swirl, parameterName=Swirl Radial Falloff Mask Length}, Value=@n53.Vector Converted to Angle) {constMax=0, constMin=0.95}
  n55 = Multiply(A=OneMinus(None=@n53.Vector Converted to Angle), B=Constant() {r=-0.125})
  n56 = Multiply(A=@n53.Vector Converted to Angle, B=Constant() {r=0.4226})
  n57 = Multiply(
   A=@n54
   B=MaterialFunctionCall(
    UVs=NamedRerouteUsage()
    Size=ScalarParameter() {defaultValue=0.6, group=Swirl, parameterName=Swirl Radial Falloff Mask Radius}
    GlowMin=@n55
    GlowMax=@n56
   ) {materialFunction=&MF_UI_Circle}.Outline
  )
  NamedRerouteDeclaration(None=Multiply(A=NamedRerouteUsage(), B=Saturate(None=@n57))) {name=RadialGlow, nodeColor=<rgba 1 0 0.375 1>}
  NamedRerouteDeclaration(None=TextureObject() {texture=&T_UI_HabaneroWaves}) {name=InnerFogPattern_Texture, nodeColor=<rgba 0 0.6953 1 1>}
  n58 = MaterialFunctionCall(UVs=NamedRerouteUsage(), Size=Constant() {r=0.8}, GlowMin=@n55, GlowMax=@n56) {materialFunction=&MF_UI_Circle}
  NamedRerouteDeclaration(None=Saturate(None=Multiply(A=@n54, B=@n58.Outline))) {name=RadialGlowBigger, nodeColor=<rgba 1 0.8906 0 1>}
  NamedRerouteDeclaration(None=Clamp(None=@n41) {maxDefault=0.5, minDefault=0.2}) {name=Swirl Color, nodeColor=<rgba 0 0 0 1>}
  n59 = MaterialFunctionCall(
   DistanceField=NamedRerouteDeclaration(None=@n50.SDFOutline) {name=OutlineSDF, nodeColor=<rgba 1 0 0.5391 1>}
   Threshold=Constant() {r=0.01}
   Smoothing=Constant() {r=0.01}
  ) {materialFunction=&MF_UI_AntiAliasing}
  n60 = ScalarParameter() {group=Outline, parameterName=2 - GradientStart}
  n61 = ScalarParameter() {defaultValue=0.7, group=Outline, parameterName=3 - GradientEnd}
  n62 = Constant() {r=0.75}
  NamedRerouteDeclaration(None=Multiply(
    A=NamedRerouteDeclaration(None=@n59.Inner) {name=OutlineAlpha, nodeColor=<rgba 0 0.5078 1 1>}
    B=MaterialFunctionCall(GradientStart=@n60, GradientEnd=@n61, RotationAngle=@n62) {materialFunction=&MF_UI_LinearGradient}.Result
   )) {name=Outline_Gradient, nodeColor=<rgba 0.8281 1 0 1>}
  n63 = ScalarParameter() {defaultValue=1, group=Swirl, parameterName=SwirlNoiseAmount}
  NamedRerouteDeclaration(None=SmoothStep(Min=OneMinus(None=@n63), Value=ComponentMask(None=@n18.RG) {g=true})) {name=Noise, nodeColor=<rgba 0.3125 0 1 1>}
  NamedRerouteDeclaration(None=Multiply(A=Clamp(None=@n35) {minDefault=0.5}, B=@n33)) {name=ButtonShine, nodeColor=<rgba 0 0.8594 1 1>}
  n64 = NamedRerouteUsage()
  n65 = NamedRerouteUsage()
  n66 = NamedRerouteUsage()
  n67 = NamedRerouteUsage()
  n68 = Max(A=Max(A=NamedRerouteUsage(), B=LinearInterpolate(A=@n65, B=@n66, Alpha=@n67)), B=NamedRerouteUsage())
  n69 = NamedRerouteUsage()
  n70 = NamedRerouteUsage()
  n71 = CurveAtlasRowParameter(CurveTime=Max(A=LinearInterpolate(A=@n68, B=@n69, Alpha=@n70), B=NamedRerouteUsage())) {
   atlas=/Game/UI/Frontend/Lobby/Habanero/Habanero_Materials/Curves/CA_PowerRank
   curve=/Game/UI/Frontend/Lobby/Habanero/Habanero_Materials/Curves/CC_PowerRank_Hyper
   defaultValue=3
   parameterName=ColorGradient
  }
  n72 = Max(A=Max(A=NamedRerouteUsage(), B=NamedRerouteUsage()), B=NamedRerouteUsage())
  n73 = NamedRerouteUsage()
  n74 = Max(A=Max(A=Max(A=@n72, B=@n73), B=NamedRerouteUsage()), B=NamedRerouteUsage())
  Saturate(None=Add(A=NamedRerouteUsage(), B=LinearInterpolate(A=@n64, B=@n71, Alpha=@n74)))
  n75 = NamedRerouteUsage()
  Saturate(None=Max(A=Max(A=@n74, B=@n75), B=NamedRerouteUsage()))
 ]
}
