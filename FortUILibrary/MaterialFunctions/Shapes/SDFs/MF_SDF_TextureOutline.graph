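# MF_SDF_TextureOutline is a UI library material function that turns a -1 to 1 SDF texture mask into an outline SDF for further operations.
# Thickness sets the band's half-width, clamped to a limit that shrinks to 0.5 as Alignment reaches fully inner or outer.
# Alignment times that width shifts the field before an absolute-value subtraction forms the band.
# A final division, guarded against zero, rescales the result to compensate.
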MaterialFunction /FortUILibrary/MaterialFunctions/Shapes/SDFs/MF_SDF_TextureOutline
graph {
 properties {
  bExposeToLibrary=true
  description=Returns an outline/stroke from a SDF Texture Mask
  previewMaterialDomain=MD_UI
  libraryCategoriesText [UI_Library]
 }
 nodes [
  Constant() {
   desc="Controls the width of the stroke band around the shape’s edge.\r\nValue : 0 to 1\r\n\r\nThe thickness always depends on the available Distance Field range.\r\n\r\nThe value defines the stroke’s half-width in the normalized SDF range (–1 to 1).\r\nIncreasing it expands the band equally inward and outward when the Outline Alignment is 0.\r\n\r\nWARNING:\r\nWhen the alignment is shifted fully inside (–1) or outside (+1), the available SDF range on one side is reduced by half.\r\nTo prevent the stroke from extending beyond the valid –1, 1 domain, the thickness is automatically clamped to 0.5 in those extreme alignments.\r\nThis ensures the entire stroke remains visible and the fill/outside regions stay stable, avoiding grey or clipped areas."
   r=0.2
  }
  Constant() {desc="Controls the stroke’s position relative to the shape’s edge.\r\n\r\n-1: aligns the stroke fully inside the shape.\r\n0: centers the stroke on the edge.\r\n1: aligns the stroke fully outside the shape.\r\n\r\nAt extreme values (–1 or 1), the stroke’s maximum thickness is automatically clamped to 0.5 to keep it fully within the SDF’s –1 to 1 range.\r\n"}
  n1 = MaterialFunctionCall!("Texture mask (0-1 range)"=TextureSample() {samplerType=SAMPLERTYPE_LinearColor, texture=/FortUILibrary/Textures/T_UI_SDF_Ready_Shape}.R) {materialFunction=/FortUILibrary/MaterialFunctions/Shapes/Modifiers/MF_UI_SDF_RemapTextureRange}
  n2 = FunctionInput(Preview=@n1.SDF) {
   bUsePreviewValueAsDefault=true
   description="The distance field (eg. SDF shape or greyscale DF texture) to generate an outline from."
   inputName="SDF Texture Mask (-1 to 1)"
   inputType=FunctionInput_Scalar
   previewValue=<0 0 0 0>
  }
  n3 = FunctionInput() {
   bUsePreviewValueAsDefault=true
   desc="Controls the stroke’s position relative to the shape’s edge.\r\n\r\n-1: aligns the stroke fully inside the shape.\r\n0: centers the stroke on the edge.\r\n1: aligns the stroke fully outside the shape.\r\n\r\nAt extreme values (–1 or 1), the stroke’s maximum thickness is automatically clamped to 0.5 to keep it fully within the SDF’s –1 to 1 range.\r\n"
   description="The alignment of the outline (0.0 = center, -1.0 = inner, 1.0 = outer)."
   inputName=Alignment
   inputType=FunctionInput_Scalar
   sortPriority=2
   previewValue=<0 0 0 0>
  }
  n4 = NamedRerouteDeclaration(None=@n3) {name=A, nodeColor=<rgba 0 1 0.0859 1>}
  n5 = FunctionInput() {
   bUsePreviewValueAsDefault=true
   desc="Controls the width of the stroke band around the shape’s edge.\r\n\r\nValue : 0 to 1\r\n\r\nThe thickness always depends on the available Distance Field range.\r\n\r\nThe value defines the stroke’s half-width in the normalized SDF range (–1 to 1).\r\nIncreasing it expands the band equally inward and outward when the Outline Alignment is 0.\r\n\r\nWARNING:\r\nWhen the alignment is shifted fully inside (–1) or outside (+1), the available SDF range on one side is reduced by half.\r\nTo prevent the stroke from extending beyond the valid –1, 1 domain, the thickness is automatically clamped to 0.5 in those extreme alignments.\r\nThis ensures the entire stroke remains visible and the fill/outside regions stay stable, avoiding grey or clipped areas.\r\n"
   description=The thickness of the outline.
   inputName=Thickness
   inputType=FunctionInput_Scalar
   sortPriority=1
   previewValue=<0.05 0 0 1>
  }
  n6 = Add(B=Abs(None=NamedRerouteUsage()) {desc="|A| = absolute value A"}) {constA=1, desc="value range : 1 to 2\r\ntakes the absolute value of the alignment param. \r\nso −1 and +1 behave the same, then adds 1"}
  n7 = NamedRerouteDeclaration(None=Divide(B=@n6) {desc="The value of MaxW \r\nranges between 0.5 and 1"}) {
   name=MaxW
   desc="The maximum safe/allowed half-width (W) of the stroke, \r\ndepending on how far the stroke is shifted by the alignment.\r\nMaxW dynamically limits W to ensure the stroke always \r\nfits entirely inside the valid range."
   nodeColor=<rgba 0.4062 0 1 1>
  }
  n8 = NamedRerouteDeclaration(None=Min(A=NamedRerouteDeclaration(None=@n5) {name=Half Width, desc=Half-width of stroke band, nodeColor=<rgba 0.4062 0 1 1>}, B=@n7)) {name=Wc, desc=Clamped half-width, nodeColor=<rgba 0.4062 0 1 1>}
  n9 = Subtract(
   A=@n2
   B=NamedRerouteDeclaration(None=Multiply(A=@n4, B=@n8) {desc=Shift the stroke}) {name="Clamped && Shifted Width", nodeColor=<rgba 0 0 0 1>}
  )
  n10 = NamedRerouteDeclaration(None=Subtract(A=Abs(None=@n9), B=NamedRerouteUsage())) {
   name=Band
   desc="The shape’s edge is where the signed distance = 0.\r\nThe band is the region around that edge where the SDF value is close to zero (for example, between –0.3 and +0.3)."
   nodeColor=<rgba 1 0.7266 0 1>
  }
  n11 = OneMinus(None=Multiply(A=NamedRerouteUsage(), B=NamedRerouteUsage()))
  FunctionOutput(None=Divide(A=@n10, B=Max(A=@n11) {constB=0, desc=avoid divide by 0})) {description=The outline SDF Texture for any further operations., outputName=SDF Texture Outline, sortPriority=1}
  MaterialFunctionCall!(X range Y range, Gradient Height, ShowGrid?, Thickness, Color, "f(x)"=@n1.SDF) {materialFunction=/Engine/Functions/Engine_MaterialFunctions01/Debug/PlotFunctionOnGraph}
  NamedRerouteDeclaration(None=@n6) {
   name=1+ AbsA
   desc="When shifting the stroke center with alignment, part of that band moves toward one side of the SDF range (e.g., closer to –1 or +1).\r\nIf we don’t limit how wide the stroke can be after that shift, we run out of range on one side of the SDF, and that causes a flat grey/black artifacts \r\nInside the stroke/band (alignment = -1) or outside the stroke/band (alignment = 1).\r\nThat’s what 1 + |A| fixes."
   nodeColor=<rgba 0 1 0.0859 1>
  }
 ]
}
