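# Translucent UI material for the Habanero lobby frontend that draws a tunnel effect of outlined geometric shapes.
# Each of three shapes unions two box SDFs, one rotated 45 degrees, into an eight-pointed star outline.
# A Time parameter, scaled by AnimationSpeed and staggered by offsets, provides the animation phases.
# A combined shape mask and three-color lerp produce the final color, alongside a dark vertical-gradient background with a White blend and an Opacity parameter.
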Material /Game/UI/Frontend/Lobby/Habanero/Habanero_Materials/M_UI_ShapeTunnelEffect
graph {
 properties {bAutomaticallySetUsageInEditor=false, blendMode=BLEND_Translucent, materialDomain=MD_UI, usedShadingModels=Default Lit}
 nodes [
  Frac(None=Time())
  NamedRerouteDeclaration(None=MaterialFunctionCall() {materialFunction=&MF_UI_UVScaling}.Scaled To Fill) {name=NormalizedUV, nodeColor=<rgba 0.0312 1 0 1>}
  n1 = ScalarParameter() {defaultValue=0.4, parameterName=Time}
  NamedRerouteDeclaration(None=@n1) {name=Time, nodeColor=<rgba 1 0 0.8203 1>}
  n2 = NamedRerouteUsage()
  NamedRerouteDeclaration(None=@n2) {name=Time_01, nodeColor=<rgba 0.1953 1 0 1>}
  NamedRerouteDeclaration(None=Add(A=@n2, B=ScalarParameter() {defaultValue=0.4, parameterName=TimeValue_05}) {constB=0.9}) {name=Time_03, nodeColor=<rgba 0.8047 0 1 1>}
  n3 = NamedRerouteUsage()
  n4 = NamedRerouteUsage()
  n5 = MaterialFunctionCall(UVs=NamedRerouteUsage(), "Rotation Angle (0-1)"=Constant() {r=0.125}) {materialFunction=&CustomRotator}
  n6 = NamedRerouteUsage()
  n7 = MaterialFunctionCall!(
   "SDF Texture Mask (-1 to 1)"=Min(
    A=MaterialFunctionCall!(UVs=@n3, Size=@n4, CornerRoundness, KeepAspectRatio, PixelSize) {materialFunction=/FortUILibrary/MaterialFunctions/Shapes/SDFs/MF_SDF_Box}.SDF
    B=MaterialFunctionCall(UVs=@n5.Rotated Values, Size=@n6) {materialFunction=&MF_SDF_Box}.SDF
   )
   Thickness=Constant() {r=-0.02}
   Alignment=Constant() {r=-1}
  ) {materialFunction=/FortUILibrary/MaterialFunctions/Shapes/SDFs/MF_SDF_TextureOutline}
  NamedRerouteDeclaration(None=Multiply(A=SmoothStep(Value=@n7.SDF Texture Outline) {constMax=0, constMin=0.1}, B=NamedRerouteUsage())) {name=Shape_01, nodeColor=<rgba 1 0 0.3984 1>}
  n8 = NamedRerouteUsage()
  n9 = NamedRerouteUsage()
  n10 = MaterialFunctionCall(UVs=NamedRerouteUsage(), "Rotation Angle (0-1)"=Constant() {r=0.125}) {materialFunction=&CustomRotator}
  n11 = NamedRerouteUsage()
  n12 = MaterialFunctionCall(
   "SDF Texture Mask (-1 to 1)"=Min(
    A=MaterialFunctionCall(UVs=@n8, Size=@n9) {materialFunction=&MF_SDF_Box}.SDF
    B=MaterialFunctionCall(UVs=@n10.Rotated Values, Size=@n11) {materialFunction=&MF_SDF_Box}.SDF
   )
   Thickness=Constant() {r=-0.02}
   Alignment=Constant() {r=-1}
  ) {materialFunction=&MF_SDF_TextureOutline}
  NamedRerouteDeclaration(None=Multiply(A=SmoothStep(Value=@n12.SDF Texture Outline) {constMax=0, constMin=0.1}, B=NamedRerouteUsage())) {name=Shape_02, nodeColor=<rgba 1 0 0.3968 1>}
  n13 = NamedRerouteUsage()
  n14 = NamedRerouteUsage()
  n15 = MaterialFunctionCall(UVs=NamedRerouteUsage(), "Rotation Angle (0-1)"=Constant() {r=0.125}) {materialFunction=&CustomRotator}
  n16 = NamedRerouteUsage()
  n17 = MaterialFunctionCall(
   "SDF Texture Mask (-1 to 1)"=Min(
    A=MaterialFunctionCall(UVs=@n13, Size=@n14) {materialFunction=&MF_SDF_Box}.SDF
    B=MaterialFunctionCall(UVs=@n15.Rotated Values, Size=@n16) {materialFunction=&MF_SDF_Box}.SDF
   )
   Thickness=Constant() {r=-0.02}
   Alignment=Constant() {r=-1}
  ) {materialFunction=&MF_SDF_TextureOutline}
  NamedRerouteDeclaration(None=Multiply(A=SmoothStep(Value=@n17.SDF Texture Outline) {constMax=0, constMin=0.1}, B=NamedRerouteUsage())) {name=Shape_03, nodeColor=<rgba 1 0 0.3968 1>}
  n18 = Constant3Vector() {constant=<rgba 0.0152 0.0006 0.0222 1>}
  n19 = Constant3Vector() {constant=<rgba 0.0137 0.0009 0.0027 1>}
  n20 = MaterialFunctionCall() {materialFunction=&LinearGradient}
  NamedRerouteDeclaration(None=LinearInterpolate(A=LinearInterpolate(A=@n18, B=@n19, Alpha=@n20.VGradient), B=Constant() {r=1}, Alpha=ScalarParameter() {parameterName=White})) {name=Background, nodeColor=<rgba 0 1 0.8594 1>}
  NamedRerouteDeclaration(None=Multiply(A=@n1, B=ScalarParameter() {defaultValue=0.7, parameterName=AnimationSpeed})) {name=AnimationSpeed, nodeColor=<rgba 1 0 0.4219 1>}
  n21 = NamedRerouteUsage()
  n22 = NamedRerouteUsage()
  NamedRerouteDeclaration(None=LinearInterpolate(A=LinearInterpolate(A=@n21, Alpha=@n22), Alpha=NamedRerouteUsage())) {name=ShapeMask, nodeColor=<rgba 1 0 0.375 1>}
  NamedRerouteDeclaration(None=Add(A=@n2, B=ScalarParameter() {defaultValue=0.2, parameterName=TimeValue_01}) {constB=0.2}) {name=Time_02, nodeColor=<rgba 0 0.8359 1 1>}
  n23 = NamedRerouteUsage()
  NamedRerouteDeclaration(None=MaterialFunctionCall(
    A=NamedRerouteUsage()
    B=VectorParameter() {parameterName=Color_02, defaultValue=<rgba 0.8796 0.008 0 1>}.RGB
    C=Constant() {r=1}
    Alpha=SmoothStep(Value=@n23) {constMin=0.1216}
   ) {materialFunction=&Lerp_3Color}.Result) {name=FinalColor, nodeColor=<rgba 0 0 0 1>}
  NamedRerouteUsage()
  ScalarParameter() {defaultValue=1, parameterName=Opacity}
 ]
}
