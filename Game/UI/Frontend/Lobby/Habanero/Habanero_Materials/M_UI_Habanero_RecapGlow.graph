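# Additive UI material for a soft two-color glow, named for the Habanero recap screen in the frontend lobby.
# A reflected gradient blends Color A and Color B.
# The glow mask multiplies a horizontal band, made from perspective-function UVs, by softened box and circle SDF falloffs sharpened with powers.
# The Glow Y Movement scalar offsets only the band's UVs vertically, sliding it within the fixed shape masks.
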Material /Game/UI/Frontend/Lobby/Habanero/Habanero_Materials/M_UI_Habanero_RecapGlow
graph {
 properties {bAutomaticallySetUsageInEditor=false, blendMode=BLEND_Additive, materialDomain=MD_UI, usedShadingModels=Default Lit}
 nodes [
  Constant()
  LinearInterpolate(
   A=VectorParameter() {parameterName=Color A, defaultValue=<rgba 1 0 0.087 1>}.RGB
   B=VectorParameter() {parameterName=Color B, defaultValue=<rgba 1 0 0.5184 1>}.RGB
   Alpha=MaterialFunctionCall!(UVs, Offset, RotationCenter, RotationAngle) {materialFunction=/FortUILibrary/MaterialFunctions/Gradients/MF_UI_ReflectedGradient}.Result
  )
  n1 = ScalarParameter() {defaultValue=1, parameterName=Glow Y Movement}
  n2 = Add(A=TextureCoordinate(), B=AppendVector(A=Constant(), B=LinearInterpolate(Alpha=@n1) {constA=0.3, constB=-0.3}))
  n3 = Constant()
  n4 = Subtract(A=ComponentMask(None=MaterialFunctionCall!(UV=@n2, Perspective Rotation, Perspective Amount=@n3) {materialFunction=/FortUILibrary/MaterialFunctions/Basics/MF_UI_Perspective}.Perspective UVs) {g=true}) {constB=0.5}
  n5 = OneMinus(None=Abs(None=Multiply(A=@n4) {constB=2}))
  n6 = Constant2Vector() {g=1, r=0.5}
  n7 = Multiply(A=MaterialFunctionCall(
    DistanceField=MaterialFunctionCall(Size=@n6) {materialFunction=&MF_UI_Box}.SDF
    Smoothing=Constant() {r=1}
   ) {materialFunction=&MF_UI_AntiAliasing}.Inner) {constB=3}
  n8 = Multiply(A=Saturate(None=Power(Base=@n5) {constExponent=4}), B=Saturate(None=Power(Base=@n7) {constExponent=3}))
  n9 = MaterialFunctionCall(Size=Constant() {r=0.6}, OutlineThickness=Constant() {r=0.1}) {materialFunction=&MF_UI_Circle}
  n10 = OneMinus(None=@n9.SDF)
  n11 = OneMinus(None=@n9.SDFOutline)
  Multiply(A=Multiply(A=Saturate(None=@n8), B=Power(Base=@n10) {constExponent=9}), B=Saturate(None=Power(Base=@n11) {constExponent=5}))
  TextureCoordinate()
 ]
}
