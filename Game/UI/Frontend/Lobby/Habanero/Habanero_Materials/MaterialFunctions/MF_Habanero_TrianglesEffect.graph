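# UI material function for the Habanero lobby frontend that outputs an animated triangle-pattern mask.
# It builds centre-scaled, rotated UI UVs and pixel-size ratios, with scale, offset and panner-speed inputs driving the scrolling texture sample.
# The texture's red channel feeds a SmoothStep whose bounds lerp from min, max and EdgeSharp inputs, with an inverted vertical UV gradient declared as Alpha.
# The AspectRatio divide uses width for both operands.
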MaterialFunction /Game/UI/Frontend/Lobby/Habanero/Habanero_Materials/MaterialFunctions/MF_Habanero_TrianglesEffect
graph {nodes [
  n1 = NamedRerouteUsage()
  n2 = FunctionInput() {bUsePreviewValueAsDefault=true, inputName=tri-L-Scale, inputType=FunctionInput_Scalar, previewValue=<0.98 0 0 1>}
  n3 = Multiply(A=NamedRerouteUsage(), B=@n2)
  n4 = Constant()
  n5 = FunctionInput() {bUsePreviewValueAsDefault=true, inputName=tri-L-Offset, inputType=FunctionInput_Scalar, previewValue=<0.98 0 0 1>}
  n6 = FunctionInput() {bUsePreviewValueAsDefault=true, inputName=PannerSpeed_X, inputType=FunctionInput_Scalar, previewValue=<0.98 0 0 1>}
  n7 = FunctionInput() {bUsePreviewValueAsDefault=true, inputName=PannerSpeed_Y, inputType=FunctionInput_Scalar, previewValue=<0.98 0 0 1>}
  n8 = Panner(Coordinate=Add(A=Multiply(A=@n1, B=@n3), B=AppendVector(A=@n4, B=@n5)), Speed=AppendVector(A=@n6, B=@n7)) {speedX=-0.5}
  n9 = FunctionInput(Preview=TextureObject() {samplerType=SAMPLERTYPE_LinearColor, texture=/Game/UI/Foundation/Items/Textures/T-Card-Hero-Anim-Triangle}) {bUsePreviewValueAsDefault=true, inputName=Texture, inputType=FunctionInput_Texture2D, previewValue=<0.98 0 0 1>}
  n10 = FunctionInput() {bUsePreviewValueAsDefault=true, inputName=min, inputType=FunctionInput_Scalar, previewValue=<0.98 0 0 1>}
  n11 = NamedRerouteUsage()
  n12 = FunctionInput() {bUsePreviewValueAsDefault=true, inputName=max, inputType=FunctionInput_Scalar, previewValue=<0.35 0 0 1>}
  n13 = FunctionInput(Preview=Constant() {r=2}) {bUsePreviewValueAsDefault=true, inputName=EdgeSharp, inputType=FunctionInput_Scalar, previewValue=<2 0 0 1>}
  FunctionOutput(None=MaterialFunctionCall(
    Alpha=TextureSample(UVs=@n8, Tex=@n9) {samplerType=SAMPLERTYPE_LinearColor, texture=&T-Card-Hero-Anim-Triangle}.R
    Min=LinearInterpolate(B=@n10, Alpha=@n11) {constA=0.25, constB=0.25}
    Max=LinearInterpolate(A=@n12, B=@n13, Alpha=@n11) {constA=0.1, constB=2}
   ) {materialFunction=&SmoothStep}.Result)
  n14 = MaterialFunctionCall(Float2=MaterialFunctionCall() {materialFunction=&GetUserInterfaceUV}.Pixel Size) {materialFunction=&BreakOutFloat2Components}
  NamedRerouteDeclaration(None=Divide(A=@n14.R, B=@n14.R) {constA=537, constB=1015}) {name=AspectRatio, nodeColor=<rgba 0.2188 1 0 1>}
  n15 = MaterialFunctionCall() {materialFunction=&GetUserInterfaceUV}
  n16 = MaterialFunctionCall(Float2=@n15.Normalized UV) {materialFunction=&BreakOutFloat2Components}
  n17 = NamedRerouteUsage()
  n18 = MaterialFunctionCall(X=Multiply(A=@n16.R, B=@n17), Y=Divide(A=@n16.G, B=@n17)) {materialFunction=&MakeFloat2}
  n19 = Constant() {r=0.1}
  NamedRerouteDeclaration(None=MaterialFunctionCall(
    UVs=MaterialFunctionCall(UVs=@n18.Result, Texture Scale=@n19) {materialFunction=&ScaleUVsByCenter}.UVs
    "Rotation Angle (0-1)"=ScalarParameter() {defaultValue=0.25, parameterName=Rotation}
   ) {materialFunction=&CustomRotator}.Rotated Values) {name=TriangleUV, nodeColor=<rgba 1 0.1641 0 1>}
  n20 = MaterialFunctionCall(Float2=@n15.Pixel Size) {materialFunction=&BreakOutFloat2Components}
  n21 = Max(B=@n20.G) {constA=0.1}
  NamedRerouteDeclaration(None=AppendVector(A=Divide(A=@n20.R, B=@n21), B=Constant() {r=1})) {name=PixelSize, nodeColor=<rgba 0 1 0.7188 1>}
  n22 = TextureCoordinate()
  NamedRerouteDeclaration(None=OneMinus(None=ComponentMask(None=@n22) {g=true})) {name=Alpha, nodeColor=<rgba 0 1 0.0625 1>}
 ]}
